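# Player animation layer Blueprint that poses a rider on the Labrador mount, reusing some Wolf rider assets.
# It copies movement state from the mount's AnimBP to drive locomotion state machines, with upper-body, lean, additive and aim-offset layering.
# A Chooser picks locomotion sequences, and mount curve values set explicit playback time.
# Leans are interpolated from body yaw offset, and boosting plays rider and mount montages linked by montage sync.
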AnimBlueprint /RidingContent/Animation/HardSaddle/Labrador/LabradorPlayerLayerAnimBP
target_skeleton=&Fortnite_M_Avg_Player_Skeleton
graphs [
 AnimGraph {nodes [Output Pose() {layerGroup=None}]}
 CorrectiveLeanPoses {nodes [
   n1 = Blend Poses by bool(
    BlendPose_0=Blendspace Player 'Labrador_Lean_Right_1DBS'!(X=Get CurrentLeanAngle() -> (CurrentLeanAngle:double)) -> (Pose) {blendSpace=&/Labrador_Lean_Right_1DBS}
    BlendPose_1=Blendspace Player 'Labrador_Lean_Left_1DBS'!(X=Get AlphaValue() -> (AlphaValue:float)) -> (Pose) {blendSpace=&/Labrador_Lean_Left_1DBS}
    BlendTime_0=0.5
    BlendTime_1=0.5
    bActiveValue=Get IsLeaningRight() -> (IsLeaningRight:bool)
   )
   n2 = Use cached pose 'PreCorrectiveLeanPose'()
   n3 = Get AlphaValue()
   Output Pose(Result=Two Way Blend(
     A=Apply Mesh Space Additive!(Base:Pose Link Structure=@n2, Additive:Pose Link Structure=@n1, Alpha:float=@n3) -> (Pose:Pose Link Structure) {is=title AnimGraphNode_ApplyMeshSpaceAdditive}
     B=Use cached pose 'PreCorrectiveLeanPose'()
     Alpha=Get DisableLeanWeight() -> (DisableLeanWeight:double)
    ))
   Save cached pose 'PreCorrectiveLeanPose'(Pose=Input Pose() {name=InPose_1}) {cacheName=PreCorrectiveLeanPose}
  ]}
 EventGraph {nodes [
   n1 = Event BlueprintUpdateAnimation()
   > n2 = Sequence!(execute) -> (then_0, then_1, then_2:Exec)
   > CalculateLeans!(execute, self, DeltaTime:double=@n1.DeltaTimeX) -> (then)
   n3 = Branch(execute=@n2.then_2, Condition=Get ShouldPlayBoostAnim() -> (ShouldPlayBoostAnim:bool))
   > Do Once(Reset=@n6.then)
   > Play Montage!(
    execute:Exec
    InSkeletalMeshComponent:Skeletal Mesh Component Object Reference=Get Mesh(self=Get Player Pawn Athena() -> (Player Pawn Athena:Player Pawn Athena Object Reference))
    MontageToPlay:Anim Montage Object Reference=&/Labrador_Player_Boost_N_M
    PlayRate:float=1
    StartingPosition:float=0
    StartingSection:Name=Loop
    bShouldStopAllMontages:bool=true
   ) -> (then:Exec, OnCompleted:Exec, OnBlendOut:Exec, OnInterrupted:Exec, OnNotifyBegin:Exec, OnNotifyEnd:Exec, NotifyName:Name) {is=title K2Node_PlayMontage}
   > Montage_Play!(
    execute
    self:Anim Instance Object Reference=Get As Labrador Anim BP() -> (As Labrador Anim BP:Labrador Anim BP Object Reference)
    MontageToPlay:Anim Montage Object Reference=/Labrador/Animation/Labrador_Run_Bounding_Boost_Mounted_M
    InPlayRate:float=1
    ReturnValueType:EMontagePlayReturnType Enum=MontageLength
    InTimeToStartMontageAt:float=0
    bStopAllMontages:bool=true
   ) -> (then, ReturnValue:float)
   n4 = Get Player Pawn Athena()
   > MontageSync_Follow!(
    execute
    self:Anim Instance Object Reference=GetAnimInstance(self=Get Mesh(self=@n4))
    MontageFollower:Anim Montage Object Reference=&Labrador_Player_Boost_N_M
    OtherAnimInstance:Anim Instance Object Reference=Get As Labrador Anim BP()
    MontageLeader:Anim Montage Object Reference=&Labrador_Run_Bounding_Boost_Mounted_M
   ) -> (then)
   n5 = Get Player Pawn Athena()
   Montage_Stop!(
    execute=@n3.else
    self:Anim Instance Object Reference=GetAnimInstance(self=Get Mesh(self=@n5))
    InBlendOutTime:float=0.15
    Montage:Anim Montage Object Reference=&Labrador_Player_Boost_N_M
   ) -> (then)
   > n6 = Montage_Stop(self=Get As Labrador Anim BP(), InBlendOutTime=0.15, Montage=&Labrador_Run_Bounding_Boost_Mounted_M)
   n7 = Get Is Jumping() -> (Is Jumping:bool)
   n8 = Get Is Falling() -> (Is Falling:bool)
   Set ShouldPlayBoostAnim(
    execute=@n2.then_1
    ShouldPlayBoostAnim:bool=AND Boolean!(A=Get Is Boosting() -> (Is Boosting:bool), B=NOT Boolean(A=@n8), C:bool=NOT Boolean(A=@n7)) -> (ReturnValue)
   ) -> (Output_Get:bool)
   n9 = Get Is Boosting()
   n10 = NOT Boolean(A=Get bIsRelaxedLevel1() -> (bIsRelaxedLevel1:bool))
   n11 = Get bIsTargeting()
   > Set ShouldApplyLean(ShouldApplyLean:bool=AND Boolean!(A=OR Boolean(A=@n10, B=@n11), B=NOT Boolean(A=@n9)) -> (ReturnValue)) -> (Output_Get:bool)
   n12 = Event BlueprintCollectRidableAnimBPData!() -> (OutputDelegate, then, RidableAnimInstance:Anim Instance Object Reference) {is=title K2Node_Event}
   > n13 = Cast To Labrador_AnimBP!(execute, Object=@n12.RidableAnimInstance) -> (then, CastFailed, AsLabrador Anim BP:Labrador Anim BP Object Reference) {is=title K2Node_DynamicCast}
   > Set As Labrador Anim BP(As Labrador Anim BP:Labrador Anim BP Object Reference=@n13) -> (Output_Get:Labrador Anim BP Object Reference)
   > GetMainAnimBPData_Labrador!(execute, self, AnimInstance:Object Reference=@n12.RidableAnimInstance) -> (then)
   > Set CurrentCurveValue(CurrentCurveValue:double=GetCurveValue(self=Get As Labrador Anim BP(), CurveName=Get CurrentCurve() -> (CurrentCurve:Name=None))) -> (Output_Get:double)
   > n14 = Cast To PlayerPawn_Athena(Object=GetOwningActor())
   > Set Player Pawn Athena(Player Pawn Athena:Player Pawn Athena Object Reference=@n14) -> (Output_Get:Player Pawn Athena Object Reference)
   AnimNotify_LandEntered() {is=title K2Node_Event}
   > Set JustLanded(JustLanded:bool=true) -> (Output_Get:bool)
   > Delay(Duration=1)
   > Set JustLanded(JustLanded=false)
   Event PlaceHandIKTargets() {is=title K2Node_Event}
  ]}
 GetChooserData {nodes [
   n1 = GetChooserData!() -> (then, UpdateContext:"Anim Update Context Structure (by ref)", Node:"Anim Node Reference Structure (by ref)") {is=title K2Node_FunctionEntry}
   > Set SequenceEvaluator(SequenceEvaluator:Sequence Evaluator Reference Structure=Convert to Sequence Evaluator!(Node:"Anim Node Reference Structure (by ref)"=@n1.Node) -> (SequenceEvaluator:Sequence Evaluator Reference Structure, Result:bool)) -> (Output_Get:Sequence Evaluator Reference Structure)
   > n2 = Evaluate Chooser: LabradorPlayer_Locomotion_Anim_CT!(execute:Exec, Labrador_AnimBP_C:Labrador Anim BP Object Reference=Get As Labrador Anim BP()) -> (execute:Exec, ChooserPlayerSettings:Chooser Player Settings Structure, CurveData:Curve Data Structure, Result:Anim Sequence Object Reference) {is=title K2Node_EvaluateChooser2, chooser=&/LabradorPlayer_Locomotion_Anim_CT}
   > Set AnimSequence(AnimSequence:Anim Sequence Object Reference=@n2.Result) -> (Output_Get:Anim Sequence Object Reference)
   > SetSequenceWithInertialBlending!(
    execute
    UpdateContext:"Anim Update Context Structure (by ref)"=Get UpdateContext() -> (UpdateContext:"Anim Update Context Structure (by ref)")
    SequenceEvaluator:"Sequence Evaluator Reference Structure (by ref)"=Get SequenceEvaluator() -> (SequenceEvaluator:Sequence Evaluator Reference Structure)
    Sequence:Anim Sequence Base Object Reference=Get AnimSequence() -> (AnimSequence:Anim Sequence Object Reference)
    BlendTime:float=0.2
   ) -> (then, ReturnValue:Sequence Evaluator Reference Structure)
   > Return Node!(
    execute
    ReturnVal:Name=Break Curve Data!(CurveData:"Curve Data Structure (by ref)"=@n2.CurveData) -> (CurveName_2_#1:Name) {is=title K2Node_BreakStruct}
    Sequence:Anim Sequence Object Reference=Get AnimSequence()
   )
  ]}
 GetMainAnimBPData_Labrador {nodes [
   n1 = GetMainAnimBPData_Labrador!() -> (then, AnimInstance:Object Reference) {is=title K2Node_FunctionEntry}
   > n2 = Is Valid(InputObject=@n1)
   Return Node!(execute=@n2.Is Not Valid)
   > n3 = Cast To Labrador_AnimBP(Object=@n1)
   > Set Is Boosting(Is Boosting:bool=Get bIsBoosting!(self:Fort AIFauna Anim Instance Object Reference=@n3) -> (bIsBoosting:bool)) -> (Output_Get:bool)
   > Set Jump Start To Jump Loop(Jump Start To Jump Loop:bool=Get bJumpStartToJumpLoop!(self:Fort AIFauna Anim Instance Labrador Object Reference=@n3) -> (bJumpStartToJumpLoop:bool)) -> (Output_Get:bool)
   > Set Loco To Idle(Loco To Idle:bool=Get bLocoToIdle!(self:Fort AIFauna Anim Instance Labrador Object Reference=@n3) -> (bLocoToIdle:bool)) -> (Output_Get:bool)
   > Set TargetIsClose(TargetIsClose:bool=Get bTargetIsClose!(self:Fort AIFauna Anim Instance Labrador Object Reference=@n3) -> (bTargetIsClose:bool)) -> (Output_Get:bool)
   > Set IsRunning(IsRunning:bool=Get bIsRunning!(self:Fort AIFauna Anim Instance Labrador Object Reference=@n3) -> (bIsRunning:bool)) -> (Output_Get:bool)
   > Set State Rule Idle Moving(State Rule Idle Moving:bool=Get bIdleToLoco!(self:Fort AIFauna Anim Instance Labrador Object Reference=@n3) -> (bIdleToLoco:bool)) -> (Output_Get:bool)
   > Set State Rule Moving Idle(State Rule Moving Idle:bool=NOT Boolean(A=Get bIsRunning(self=@n3))) -> (Output_Get:bool)
   > Set Is Jumping(Is Jumping:bool=Get bIsJumping!(self:Fort Anim Instance Object Reference=@n3) -> (bIsJumping)) -> (Output_Get:bool)
   > Set State Rule Fall Idle(State Rule Fall Idle:bool=Get bTransitionToLand!(self:Fort AIFauna Anim Instance Labrador Object Reference=@n3) -> (bTransitionToLand:bool)) -> (Output_Get:bool)
   > Set Is Falling(Is Falling:bool=Get bIdleToFalling!(self:Fort AIFauna Anim Instance Labrador Object Reference=@n3) -> (bIdleToFalling:bool)) -> (Output_Get:bool)
   > Set State Rule Jump Land Moving(State Rule Jump Land Moving:bool=Get bJumpLandToMovement!(self:Fort AIFauna Anim Instance Labrador Object Reference=@n3) -> (bJumpLandToMovement:bool)) -> (Output_Get:bool)
   > Set Is Surface Swimming(Is Surface Swimming:bool=Get bIsSurfaceSwimming!(self:Fort AIFauna Anim Instance Object Reference=@n3) -> (bIsSurfaceSwimming)) -> (Output_Get:bool)
   > Set BodyRoll(BodyRoll:double=Get BodyRoll!(self:Fort AIFauna Anim Instance Object Reference=@n3) -> (BodyRoll:float)) -> (Output_Get:double)
   > Set WolfSpeed2D(WolfSpeed2D:double=Get PawnSpeed2D!(self:Fort Anim Instance Object Reference=@n3) -> (PawnSpeed2D)) -> (Output_Get:double)
   > Set Play Land Additive(Play Land Additive:bool=Get bPlayLandAdditive!(self:Fort AIFauna Anim Instance Object Reference=@n3) -> (bPlayLandAdditive:bool)) -> (Output_Get:bool)
   > Set Is Moving Backwards(Is Moving Backwards:bool=Get bIsMovingBackwards!(self:Fort AIFauna Anim Instance Object Reference=@n3) -> (bIsMovingBackwards:bool)) -> (Output_Get:bool)
   > Set Should Play Stop Backwards(Should Play Stop Backwards:bool=Get bShouldPlayStopBackwards!(self:Fort AIFauna Anim Instance Object Reference=@n3) -> (bShouldPlayStopBackwards:bool)) -> (Output_Get:bool)
   > Set Is Turning in Place(Is Turning in Place:bool=Get bIsTurningInPlace!(self:Fort AIFauna Anim Instance Object Reference=@n3) -> (bIsTurningInPlace:bool)) -> (Output_Get:bool)
   > Set Turn in Place Angle Delta(Turn in Place Angle Delta:double=Get TurnInPlaceAngleDelta!(self:Fort AIFauna Anim Instance Object Reference=@n3) -> (TurnInPlaceAngleDelta:float)) -> (Output_Get:double)
   > Set Was Turning Right(Was Turning Right:bool=Get bWasTurningRight!(self:Fort AIFauna Anim Instance Object Reference=@n3) -> (bWasTurningRight:bool)) -> (Output_Get:bool)
   > Set Turn in Place Rotation Speed(Turn in Place Rotation Speed:double=Get TurnInPlaceRotationSpeed!(self:Fort AIFauna Anim Instance Object Reference=@n3) -> (TurnInPlaceRotationSpeed:float)) -> (Output_Get:double)
   > Set Backwards Play Rate(Backwards Play Rate:double=Get BackwardsPlayRate!(self:Fort AIFauna Anim Instance Object Reference=@n3) -> (BackwardsPlayRate:float)) -> (Output_Get:double)
   > Set bIsTurnInPlaceAngleDeltaPositive(bIsTurnInPlaceAngleDeltaPositive:bool=Get bIsTurnInPlaceAngleDeltaPositive!(self:Fort AIFauna Anim Instance Object Reference=@n3) -> (bIsTurnInPlaceAngleDeltaPositive:bool)) -> (Output_Get:bool)
   > Set RidingMovePlayRate(RidingMovePlayRate:float=Get RidingMovePlayRate!(self:Fort AIFauna Anim Instance Object Reference=@n3) -> (RidingMovePlayRate:float)) -> (Output_Get:float)
   > Return Node()
  ]}
 LerpTo {nodes [
   LerpTo!() -> (then, Current:double=~, Target:double=~, DeltaTime:double=~, Speed:double=~) {is=title K2Node_FunctionEntry}
   n1 = float - float(A=Get Target() -> (Target:double), B=Get Current() -> (Current:double))
   n2 = float * float(A=Get DeltaTime() -> (DeltaTime:double), B=Get Speed() -> (Speed:double)) {is=title K2Node_PromotableOperator}
   n3 = float * float(A=@n2, B=-1)
   > Return Node!(execute, InterpResult:double=float + float(A=Get Current(), B="Clamp (Float)"(Value=@n1, Min=@n3, Max=@n2)))
  ]}
 OnRelevant_Movement {nodes [
   n1 = OnRelevant_Movement!() -> (then, Context:"Anim Update Context Structure (by ref)", Node:"Anim Node Reference Structure (by ref)") {is=title K2Node_FunctionEntry}
   > GetChooserData!(execute, self, UpdateContext:"Anim Update Context Structure (by ref)"=@n1, Node:"Anim Node Reference Structure (by ref)"=@n1.Node) -> (then, ReturnVal:Name=None, Sequence:Anim Sequence Object Reference) {is=title K2Node_CallFunction}
   > Return Node!(execute)
  ]}
 OnUpdate_Movement {nodes [
   n1 = OnUpdate_Movement!() -> (then, Context:"Anim Update Context Structure (by ref)", Node:"Anim Node Reference Structure (by ref)") {is=title K2Node_FunctionEntry}
   > Set SequenceEvaluator(SequenceEvaluator=Convert to Sequence Evaluator(Node=@n1.Node))
   > n2 = GetChooserData(
    UpdateContext=Get Context() -> (Context:"Anim Update Context Structure (by ref)")
    Node=Get Node() -> (Node:"Anim Node Reference Structure (by ref)")
   ) -> (ReturnVal=None)
   > Set CurrentCurve(CurrentCurve:Name=@n2) -> (Output_Get:Name=None)
   > SetExplicitTime!(
    execute
    SequenceEvaluator:"Sequence Evaluator Reference Structure (by ref)"=Get SequenceEvaluator()
    Time:float=Get CurrentCurveValue() -> (CurrentCurveValue:double)
   ) -> (then, ReturnValue:Sequence Evaluator Reference Structure)
  ]}
 ProceduralLeans {nodes [
   n1 = Input Pose() {name=InPose_0}
   n2 = "Transform (Modify) Bone - Bone: pelvis"!(
    ComponentPose="Transform (Modify) Bone - Bone: pelvis"!(
     ComponentPose=Local To Component(LocalPose=@n1)
     Alpha=1
     Translation_X:double=0
     Translation_Y:double=0
     Translation_Z:double=Get LeanOffset() -> (LeanOffset:double)
     Rotation:Rotator="0,0,0"
     Scale="1,1,1"
    ) -> (Pose) {translationMode=BMM_Additive, boneToModify {boneName=pelvis}, translation {z=-10}}
    Alpha=1
    Translation="0,0,0"
    Rotation_Roll:float=0
    Rotation_Pitch:float=Get CurrentLeanAngle()
    Rotation_Yaw:float=0
    Scale="1,1,1"
   ) -> (Pose) {rotationMode=BMM_Additive, boneToModify {boneName=pelvis}, rotation {roll=45}}
   n3 = Get LeanOffsetCorrectionX() -> (LeanOffsetCorrectionX:double)
   n4 = Get LeanOffsetCorrectionZ() -> (LeanOffsetCorrectionZ:double)
   n5 = "Transform (Modify) Bone - Bone: Spine_02"!(
    ComponentPose="Transform (Modify) Bone - Bone: Spine_01"!(
     ComponentPose="Transform (Modify) Bone - Bone: pelvis"(ComponentPose=@n2, Alpha=1, Translation_X=@n3, Translation_Y=0, Translation_Z=@n4, Rotation="0,0,0", Scale="1,1,1") {translationMode=BMM_Additive, boneToModify {boneName=pelvis}}
     Alpha=1
     Translation="0,0,0"
     Rotation_Roll:float=0
     Rotation_Pitch:float=Get LeanSpineChainAngle() -> (LeanSpineChainAngle:double)
     Rotation_Yaw:float=0
     Scale="1,1,1"
    ) -> (Pose) {rotationMode=BMM_Additive, boneToModify {boneName=Spine_01}, rotation {roll=45}}
    Alpha=1
    Translation="0,0,0"
    Rotation_Roll:float=0
    Rotation_Pitch:float=Get LeanSpineChainAngle()
    Rotation_Yaw:float=0
    Scale="1,1,1"
   ) -> (Pose) {rotationMode=BMM_Additive, boneToModify {boneName=Spine_02}, rotation {roll=45}}
   n6 = Get LeanSpineChainAngle()
   n7 = Get LeanSpineChainAngle()
   n8 = "Transform (Modify) Bone - Bone: spine_04"!(
    ComponentPose="Transform (Modify) Bone - Bone: spine_04"(
     ComponentPose="Transform (Modify) Bone - Bone: Spine_03"!(ComponentPose=@n5, Alpha=1, Translation="0,0,0", Rotation_Roll:float=0, Rotation_Pitch:float=@n6, Rotation_Yaw:float=0, Scale="1,1,1") -> (Pose) {rotationMode=BMM_Additive, boneToModify {boneName=Spine_03}, rotation {roll=45}}
     Alpha=1
     Translation="0,0,0"
     Rotation_Roll=0
     Rotation_Pitch=@n7
     Rotation_Yaw=0
     Scale="1,1,1"
    ) {rotationMode=BMM_Additive, boneToModify {boneName=spine_04}, rotation {roll=45}}
    Alpha=1
    Translation="0,0,0"
    Rotation_Roll:float=0
    Rotation_Pitch:float=@n7
    Rotation_Yaw:float=0
    Scale="1,1,1"
   ) -> (Pose) {rotationMode=BMM_Additive, boneToModify {boneName=spine_04}, rotation {roll=45}}
   Output Pose(Result=Component To Local(ComponentPose="Transform (Modify) Bone - Bone: spine_05"!(ComponentPose=@n8, Alpha=1, Translation="0,0,0", Rotation_Roll:float=0, Rotation_Pitch:float=@n7, Rotation_Yaw:float=0, Scale="1,1,1") -> (Pose) {rotationMode=BMM_Additive, boneToModify {boneName=spine_05}, rotation {roll=45}}))
  ]}
 ProceduralLeansGraph {nodes [
   n1 = CalculateLeans!() -> (OutputDelegate, then, DeltaTime:double=~) {is=title K2Node_CustomEvent}
   > n2 = Sequence!(execute) -> (then_0, then_1, then_2, then_3:Exec, then_4:Exec)
   n3 = float + float(
    A=float * float(A=Get AimLeanSwitchThresholdAngle() -> (AimLeanSwitchThresholdAngle:double), B=-1)
    B=Get AimLeanSwitchThresholdAngleOffset() -> (AimLeanSwitchThresholdAngleOffset:double)
   )
   n4 = Get BodyYawOffset() -> (BodyYawOffset:float)
   n5 = float + float(A=Get AimLeanSwitchThresholdAngle(), B=Get AimLeanSwitchThresholdAngleOffset())
   n6 = Get BodyYawOffset()
   > Branch(Condition=Select(Option 0:bool="float > float"(A=@n6, B=@n5), Option 1:bool="float < float"(A=@n4, B=@n3), Index=Get IsLeaningRight()) -> (ReturnValue:bool))
   > Set IsLeaningRight(IsLeaningRight:bool=NOT Boolean(A=Get IsLeaningRight())) -> (Output_Get:bool)
   n7 = "Absolute (Float)"(A=Get BodyYawOffset())
   n8 = Get Lean from Angle() -> (Lean from Angle:Curve Float Object Reference)
   n9 = Get MaxLeanAngle() -> (MaxLeanAngle:double)
   n10 = float * float(A=@n9, B=-1)
   n11 = Get IsLeaningRight()
   Set TargetLeanAngle(
    execute=@n2.then_1
    TargetLeanAngle:double=float * float(A=SelectFloat(A=@n9, B=@n10, bPickA=@n11), B=GetFloatValue!(self:Curve Float Object Reference=@n8, InTime:float=@n7) -> (ReturnValue:float))
   ) -> (Output_Get:double)
   n12 = Get BodyYawOffset()
   > Set AlphaValue(AlphaValue:float=GetFloatValue(self=Get Lean from Angle(), InTime="Absolute (Float)"(A=@n12))) -> (Output_Get:float)
   Set CurrentLeanAngle(
    execute=@n2.then_2
    CurrentLeanAngle:double=FInterpTo(Current=Get CurrentLeanAngle(), Target=Get TargetLeanAngle() -> (TargetLeanAngle:double), DeltaTime=@n1.DeltaTime, InterpSpeed=4)
   ) -> (Output_Get:double)
   n13 = Degrees To Radians!(A:double=Get CurrentLeanAngle()) -> (ReturnValue:double)
   Set LeanOffsetCorrectionX(execute=@n2.then_3, LeanOffsetCorrectionX:double=float * float(A="Sin (Radians)"(A=@n13), B=Get LeanOffset())) -> (Output_Get:double)
   n14 = "Cos (Radians)"!(A:double=@n13) -> (ReturnValue:double)
   n15 = Get LeanOffset()
   > Set LeanOffsetCorrectionZ(LeanOffsetCorrectionZ:double=float * float(A=float * float(A=@n14, B=@n15), B=-1)) -> (Output_Get:double)
   Set LeanSpineChainAngle(execute=@n2.then_4, LeanSpineChainAngle:double=float / float(A=Get CurrentLeanAngle(), B=-6)) -> (Output_Get:double)
  ]}
 VehicleFinalPoseOverride {nodes [
   Output Pose(Result=Use cached pose 'PreLeans'()) {layerGroup=VehicleOverride}
   n1 = "Transform (Modify) Bone - Bone: Root"!(
    ComponentPose=Local To Component(LocalPose=Input Pose() {name=InFinalPose})
    Alpha=1
    Translation=Get PlayerOffset() -> (PlayerOffset:Vector)
   ) -> (Pose) {
    rotationSpace=BCS_WorldSpace
    translationMode=BMM_Additive
    translationSpace=BCS_WorldSpace
    alphaBoolBlend {blendOption=Linear}
    boneToModify {boneName=Root}
   }
   n2 = "Transform (Modify) Bone - Bone: Root"!(
    ComponentPose=@n1
    Alpha=Get BodyCounterRotationAlpha() -> (BodyCounterRotationAlpha:float)
    Rotation:Rotator=Get BodyCounterRotation() -> (BodyCounterRotation:Rotator)
   ) -> (Pose) {
    rotationMode=BMM_Replace
    rotationSpace=BCS_WorldSpace
    translationSpace=BCS_WorldSpace
    alphaBoolBlend {blendOption=Linear}
    boneToModify {boneName=Root}
   }
   Save cached pose 'PreLeans'(Pose=Component To Local(ComponentPose=@n2)) {cacheName=PreLeans}
  ]}
 VehicleFullBodyOverride {nodes [
   n1 = Layered blend per bone(BasePose=Use cached pose 'PostUpperBodyPose'(), BlendPoses_0=Use cached pose 'UpperAndLowerPose'()) {
    bBlendRootMotionBasedOnRootBone=false
    bMeshSpaceRotationBlend=true
    layerSetup [{branchFilters [
       {boneName=Root}
       {blendDepth=-1, boneName=pelvis}
       {boneName=Spine_01}
       {blendDepth=-1, boneName=thigh_l}
       {blendDepth=-1, boneName=thigh_r}
       {blendDepth=-1, boneName=ik_foot_root}
      ]}]
   }
   AnimGraphNode_StateMachine_2 = CombatUpperBodyAdditive() {is=title AnimGraphNode_StateMachine, bAllowConduitEntryStates=true, bSkipFirstUpdateTransition=false, maxTransitionsPerFrame=1}
   AnimGraphNode_StateMachine_4 = EntryAdditive() {is=title AnimGraphNode_StateMachine}
   n2 = Apply Mesh Space Additive(
    Base=Apply Mesh Space Additive(Base=@n1, Additive=@AnimGraphNode_StateMachine_2, Alpha=Get NoiseAlpha() -> (NoiseAlpha:float)) {alphaBoolBlend {blendInTime=0.6, blendOutTime=0.35}}
    Additive=@AnimGraphNode_StateMachine_4
    Alpha=1
   ) {alphaBoolBlend {blendInTime=0.6, blendOutTime=0.35}}
   n3 = Get BodyYawOffset()
   AnimGraphNode_StateMachine_3 = MeleeOffset() {is=title AnimGraphNode_StateMachine}
   n4 = Blend Poses by bool(
    BlendPose_0=Use cached pose 'PostUpperBodyPose'()
    BlendPose_1=Apply Additive!(
     Base=AimOffset Player 'HardSaddleWolf_Player_AO_Yaw'!(X:float=@n3, BasePose:Pose Link Structure=@n2, Alpha:float=1) -> (Pose:Pose Link Structure) {blendSpace=/RidingContent/Animation/HardSaddle/Wolf/AO/HardSaddleWolf_Player_AO_Yaw, alphaBoolBlend {blendOption=Linear}}
     Additive=@AnimGraphNode_StateMachine_3
     bAlphaBoolEnabled:bool=Get bShouldAdjustBodyTwistForMelee() -> (bShouldAdjustBodyTwistForMelee:bool)
    ) -> (Pose) {alphaInputType=Bool, alphaBoolBlend {blendInTime=0.25, blendOption=Linear, blendOutTime=0.25}}
    BlendTime_0=0.25
    BlendTime_1=Get FalseBlendTime() -> (FalseBlendTime:float)
    bActiveValue=Get bIsRideMode() -> (bIsRideMode:bool)
   ) {blendProfile=&Fortnite_M_Avg_Player_Skeleton:FastWeapons, blendType=Linear}
   Output Pose(Result=Inertialization!(Source:Pose Link Structure=WeaponAttachmentProceduralAdjustment!(InPose:Pose Link Structure=@n4) -> (Pose:Pose Link Structure) {is=title AnimGraphNode_LinkedAnimLayer, layer=WeaponAttachmentProceduralAdjustment}) -> (Pose:Pose Link Structure) {is=title AnimGraphNode_Inertialization}) {layerGroup=VehicleOverride}
   Save cached pose 'UpperAndLowerPose'(Pose=Input Pose() {name=InPoseFullBody}) {cacheName=UpperAndLowerPose}
   n5 = Use cached pose 'FullBodyPose'()
   AnimGraphNode_StateMachine_1 = UpperBody() {is=title AnimGraphNode_StateMachine, bAllowConduitEntryStates=true, bSkipFirstUpdateTransition=false, maxTransitionsPerFrame=1}
   n6 = Layered blend per bone(
    BasePose=Blend Poses by bool(
     BlendPose_0=CorrectiveLeanPoses!(InPose_1:Pose Link Structure=@n5) -> (Pose:Pose Link Structure) {is=title AnimGraphNode_LinkedAnimLayer, layer=CorrectiveLeanPoses}
     BlendPose_1=Use cached pose 'FullBodyPose'()
     BlendTime_0=0.15
     BlendTime_1=0.3
     bActiveValue=Get ShouldApplyLean() -> (ShouldApplyLean:bool)
    )
    BlendPoses_0=@AnimGraphNode_StateMachine_1
   ) {
    bBlendRootMotionBasedOnRootBone=false
    bMeshSpaceRotationBlend=true
    layerSetup [{branchFilters [
       {boneName=Root}
       {blendDepth=-1, boneName=pelvis}
       {boneName=Spine_01}
       {blendDepth=-1, boneName=thigh_l}
       {blendDepth=-1, boneName=thigh_r}
       {blendDepth=-1, boneName=ik_foot_root}
      ]}]
   }
   Save cached pose 'PostUpperBodyPose'(Pose=@n6) {cacheName=PostUpperBodyPose}
   AnimGraphNode_StateMachine_0 = FullBody() {is=title AnimGraphNode_StateMachine, bAllowConduitEntryStates=true, bSkipFirstUpdateTransition=false}
   Save cached pose 'FullBodyPose'(Pose=@AnimGraphNode_StateMachine_0) {cacheName=FullBodyPose}
   Get CurrentLeanAngle()
   Get AlphaValue()
  ]}
 FullBody {
  entry=Idle
  path=VehicleFullBodyOverride.AnimGraphNode_StateMachine_0.FullBody
  ids=id
  nodes [
   SA_ToMoving() {is=title AnimStateAliasNode, also {aliases [Idle,JumpLand,Stop,Turn]}}
   SA_ToJumpStart() {is=title AnimStateAliasNode, also {aliases [Idle,JumpLand,Moving,Startle,Stop,Swimming,Turn]}}
   SA_ToSwimming() {is=title AnimStateAliasNode, bGlobalAlias=true}
   SA_ToJumpFall() {is=title AnimStateAliasNode, also {aliases [Idle]}}
   AnimStateNode_0 = JumpStart() {is=title AnimStateNode}
   AnimStateNode_1 = Stop() {is=title AnimStateNode}
   AnimStateNode_10 = Startle() {is=title AnimStateNode}
   AnimStateNode_2 = Idle() {
    stateEntered {guid=#2}
    stateFullyBlended {guid=#3}
    stateLeft {guid=#4}
   }
   AnimStateNode_3 = Moving() {
    is=title AnimStateNode
    stateEntered {guid=#5}
    stateFullyBlended {guid=#6}
    stateLeft {guid=#7}
   }
   AnimStateNode_4 = JumpFall() {is=title AnimStateNode}
   AnimStateNode_5 = Turn() {is=title AnimStateNode}
   AnimStateNode_6 = JumpLand() {is=title AnimStateNode, stateEntered {guid=#8, notifyName=LandEntered}}
   AnimStateNode_7 = Swimming() {is=title AnimStateNode}
   AnimStateTransitionNode_0 = AnimStateTransitionNode() {logicType=TLT_Inertialization, priorityOrder=100}
   AnimStateTransitionNode_1 = AnimStateTransitionNode() {crossfadeDuration=0.5, logicType=TLT_Inertialization}
   AnimStateTransitionNode_10 = AnimStateTransitionNode() {logicType=TLT_Inertialization, priorityOrder=2}
   AnimStateTransitionNode_11 = AnimStateTransitionNode() {bDisabled=true, blendMode=Linear, crossfadeDuration=0.3, priorityOrder=2}
   AnimStateTransitionNode_12 = AnimStateTransitionNode() {bDisabled=true, crossfadeDuration=0.4, logicType=TLT_Inertialization}
   AnimStateTransitionNode_13 = AnimStateTransitionNode() {crossfadeDuration=0.4, logicType=TLT_Inertialization}
   AnimStateTransitionNode_14 = AnimStateTransitionNode() {bAllowInertializationForSelfTransitions=true}
   AnimStateTransitionNode_15 = AnimStateTransitionNode() {bAllowInertializationForSelfTransitions=true}
   AnimStateTransitionNode_18 = AnimStateTransitionNode() {bAllowInertializationForSelfTransitions=true, logicType=TLT_Inertialization}
   AnimStateTransitionNode_2 = AnimStateTransitionNode() {bDisabled=true, crossfadeDuration=0.35, logicType=TLT_Inertialization, priorityOrder=3}
   AnimStateTransitionNode_23 = AnimStateTransitionNode() {bAllowInertializationForSelfTransitions=true}
   AnimStateTransitionNode_24 = AnimStateTransitionNode() {bAllowInertializationForSelfTransitions=true, bAutomaticRuleBasedOnSequencePlayerInState=true}
   AnimStateTransitionNode_3 = AnimStateTransitionNode() {bAutomaticRuleBasedOnSequencePlayerInState=true, crossfadeDuration=0.4, priorityOrder=3}
   AnimStateTransitionNode_4 = AnimStateTransitionNode() {logicType=TLT_Inertialization, priorityOrder=100}
   AnimStateTransitionNode_5 = AnimStateTransitionNode() {bAutomaticRuleBasedOnSequencePlayerInState=true, crossfadeDuration=0.3, logicType=TLT_Inertialization, priorityOrder=100}
   AnimStateTransitionNode_6 = AnimStateTransitionNode() {crossfadeDuration=0.4, logicType=TLT_Inertialization}
   AnimStateTransitionNode_7 = AnimStateTransitionNode() {logicType=TLT_Inertialization}
   AnimStateTransitionNode_8 = AnimStateTransitionNode() {bDisabled=true}
   AnimStateTransitionNode_9 = AnimStateTransitionNode() {crossfadeDuration=0.15, logicType=TLT_Inertialization}
  ]
  links [
   {
    dst=Moving
    rule=VehicleFullBodyOverride.AnimGraphNode_StateMachine_0.FullBody.AnimStateTransitionNode_14.Transition
    src=Idle
    via=AnimStateTransitionNode_14
   }
   {
    dst=Turn
    rule=VehicleFullBodyOverride.AnimGraphNode_StateMachine_0.FullBody.AnimStateTransitionNode_8.Transition
    src=Idle
    via=AnimStateTransitionNode_8
   }
   {
    dst=JumpLand
    rule=VehicleFullBodyOverride.AnimGraphNode_StateMachine_0.FullBody.AnimStateTransitionNode_4.Transition
    src=JumpFall
    via=AnimStateTransitionNode_4
   }
   {
    dst=Moving
    rule=VehicleFullBodyOverride.AnimGraphNode_StateMachine_0.FullBody.AnimStateTransitionNode_9.Transition
    src=JumpFall
    via=AnimStateTransitionNode_9
   }
   {
    dst=Idle
    rule=VehicleFullBodyOverride.AnimGraphNode_StateMachine_0.FullBody.AnimStateTransitionNode_5.Transition
    src=JumpLand
    via=AnimStateTransitionNode_5
   }
   {
    dst=Moving
    rule=VehicleFullBodyOverride.AnimGraphNode_StateMachine_0.FullBody.AnimStateTransitionNode_10.Transition
    src=JumpLand
    via=AnimStateTransitionNode_10
   }
   {
    dst=Moving
    rule=VehicleFullBodyOverride.AnimGraphNode_StateMachine_0.FullBody.AnimStateTransitionNode_15.Transition
    src=JumpLand
    via=AnimStateTransitionNode_15
   }
   {
    dst=JumpFall
    rule=VehicleFullBodyOverride.AnimGraphNode_StateMachine_0.FullBody.AnimStateTransitionNode_1.Transition
    src=JumpStart
    via=AnimStateTransitionNode_1
   }
   {
    dst=Idle
    rule=VehicleFullBodyOverride.AnimGraphNode_StateMachine_0.FullBody.AnimStateTransitionNode_18.Transition
    src=Moving
    via=AnimStateTransitionNode_18
   }
   {
    dst=Stop
    rule=VehicleFullBodyOverride.AnimGraphNode_StateMachine_0.FullBody.AnimStateTransitionNode_2.Transition
    src=Moving
    via=AnimStateTransitionNode_2
   }
   {
    dst=JumpFall
    rule=VehicleFullBodyOverride.AnimGraphNode_StateMachine_0.FullBody.AnimStateTransitionNode_13.Transition
    src=SA_ToJumpFall
    via=AnimStateTransitionNode_13
   }
   {
    dst=JumpStart
    rule=VehicleFullBodyOverride.AnimGraphNode_StateMachine_0.FullBody.AnimStateTransitionNode_0.Transition
    src=SA_ToJumpStart
    via=AnimStateTransitionNode_0
   }
   {
    dst=Startle
    rule=VehicleFullBodyOverride.AnimGraphNode_StateMachine_0.FullBody.AnimStateTransitionNode_23.AnimationTransitionGraph_0
    src=SA_ToMoving
    via=AnimStateTransitionNode_23
   }
   {
    dst=Swimming
    rule=VehicleFullBodyOverride.AnimGraphNode_StateMachine_0.FullBody.AnimStateTransitionNode_6.Transition
    src=SA_ToSwimming
    via=AnimStateTransitionNode_6
   }
   {
    dst=Moving
    rule=VehicleFullBodyOverride.AnimGraphNode_StateMachine_0.FullBody.AnimStateTransitionNode_24.AnimationTransitionGraph_0
    src=Startle
    via=AnimStateTransitionNode_24
   }
   {
    dst=Idle
    rule=VehicleFullBodyOverride.AnimGraphNode_StateMachine_0.FullBody.AnimStateTransitionNode_3.Transition
    src=Stop
    via=AnimStateTransitionNode_3
   }
   {
    dst=Moving
    rule=VehicleFullBodyOverride.AnimGraphNode_StateMachine_0.FullBody.AnimStateTransitionNode_11.Transition
    src=Stop
    via=AnimStateTransitionNode_11
   }
   {
    dst=Turn
    rule=VehicleFullBodyOverride.AnimGraphNode_StateMachine_0.FullBody.AnimStateTransitionNode_12.Transition
    src=Stop
    via=AnimStateTransitionNode_12
   }
   {
    dst=Moving
    rule=VehicleFullBodyOverride.AnimGraphNode_StateMachine_0.FullBody.AnimStateTransitionNode_7.Transition
    src=Swimming
    via=AnimStateTransitionNode_7
   }
  ]
 }
 JumpStart {
  path=VehicleFullBodyOverride.AnimGraphNode_StateMachine_0.FullBody.AnimStateNode_0.JumpStart
  nodes [Output Animation Pose(Result=Sequence Player 'Labrador_Player_Jump_Start'() {bLoopAnimation=false, sequence=&/Labrador_Player_Jump_Start})]
 }
 Stop {
  path=VehicleFullBodyOverride.AnimGraphNode_StateMachine_0.FullBody.AnimStateNode_1.Stop
  nodes [
   AnimGraphNode_StateMachine_0 = StopTransition() {is=title AnimGraphNode_StateMachine, bAllowConduitEntryStates=true, bSkipFirstUpdateTransition=false, maxTransitionsPerFrame=1}
   Output Animation Pose(Result=@AnimGraphNode_StateMachine_0)
  ]
 }
 StopTransition {
  entry=Conduit
  path=VehicleFullBodyOverride.AnimGraphNode_StateMachine_0.FullBody.AnimStateNode_1.Stop.AnimGraphNode_StateMachine_0.StopTransition
  ids=id
  nodes [
   AnimStateConduitNode_0 = Conduit()
   AnimStateNode_0 = FeetInAir() {is=title AnimStateNode}
   AnimStateNode_1 = FrontFeetPlanted() {is=title AnimStateNode}
   AnimStateNode_2 = BackFeetPlanted() {is=title AnimStateNode}
   AnimStateNode_3 = Backwards() {is=title AnimStateNode}
   AnimStateTransitionNode_0 = AnimStateTransitionNode() {logicType=TLT_Inertialization, priorityOrder=0}
   AnimStateTransitionNode_3 = AnimStateTransitionNode() {logicType=TLT_Inertialization}
   AnimStateTransitionNode_4 = AnimStateTransitionNode() {logicType=TLT_Inertialization, priorityOrder=2}
   AnimStateTransitionNode_5 = AnimStateTransitionNode() {logicType=TLT_Inertialization, priorityOrder=100}
  ]
  links [
   {
    dst=BackFeetPlanted
    rule=VehicleFullBodyOverride.AnimGraphNode_StateMachine_0.FullBody.AnimStateNode_1.Stop.AnimGraphNode_StateMachine_0.StopTransition.AnimStateTransitionNode_5.Transition
    src=Conduit
    via=AnimStateTransitionNode_5
   }
   {
    dst=Backwards
    rule=VehicleFullBodyOverride.AnimGraphNode_StateMachine_0.FullBody.AnimStateNode_1.Stop.AnimGraphNode_StateMachine_0.StopTransition.AnimStateTransitionNode_0.Transition
    src=Conduit
    via=AnimStateTransitionNode_0
   }
   {
    dst=FeetInAir
    rule=VehicleFullBodyOverride.AnimGraphNode_StateMachine_0.FullBody.AnimStateNode_1.Stop.AnimGraphNode_StateMachine_0.StopTransition.AnimStateTransitionNode_3.Transition
    src=Conduit
    via=AnimStateTransitionNode_3
   }
   {
    dst=FrontFeetPlanted
    rule=VehicleFullBodyOverride.AnimGraphNode_StateMachine_0.FullBody.AnimStateNode_1.Stop.AnimGraphNode_StateMachine_0.StopTransition.AnimStateTransitionNode_4.Transition
    src=Conduit
    via=AnimStateTransitionNode_4
   }
  ]
 }
 Conduit {
  path=VehicleFullBodyOverride.AnimGraphNode_StateMachine_0.FullBody.AnimStateNode_1.Stop.AnimGraphNode_StateMachine_0.StopTransition.AnimStateConduitNode_0.Conduit
  nodes [Result(bCanEnterTransition=true)]
 }
 FeetInAir {
  path=VehicleFullBodyOverride.AnimGraphNode_StateMachine_0.FullBody.AnimStateNode_1.Stop.AnimGraphNode_StateMachine_0.StopTransition.AnimStateNode_0.FeetInAir
  nodes [Output Animation Pose(Result=Sequence Player 'HardSaddleWolf_Player_Stop_FeetInAir'() {bLoopAnimation=false, sequence=/RidingContent/Animation/HardSaddle/Wolf/HardSaddleWolf_Player_Stop_FeetInAir})]
 }
 FrontFeetPlanted {
  path=VehicleFullBodyOverride.AnimGraphNode_StateMachine_0.FullBody.AnimStateNode_1.Stop.AnimGraphNode_StateMachine_0.StopTransition.AnimStateNode_1.FrontFeetPlanted
  nodes [Output Animation Pose(Result=Sequence Player 'HardSaddleWolf_Player_Stop_FrontFeetPlanted'() {bLoopAnimation=false, sequence=/RidingContent/Animation/HardSaddle/Wolf/HardSaddleWolf_Player_Stop_FrontFeetPlanted})]
 }
 BackFeetPlanted {
  path=VehicleFullBodyOverride.AnimGraphNode_StateMachine_0.FullBody.AnimStateNode_1.Stop.AnimGraphNode_StateMachine_0.StopTransition.AnimStateNode_2.BackFeetPlanted
  nodes [Output Animation Pose(Result=Sequence Player 'HardSaddleWolf_Player_Stop_BackFeetPlanted'() {bLoopAnimation=false, sequence=/RidingContent/Animation/HardSaddle/Wolf/HardSaddleWolf_Player_Stop_BackFeetPlanted})]
 }
 Backwards {
  path=VehicleFullBodyOverride.AnimGraphNode_StateMachine_0.FullBody.AnimStateNode_1.Stop.AnimGraphNode_StateMachine_0.StopTransition.AnimStateNode_3.Backwards
  nodes [
   AnimGraphNode_StateMachine_0 = StopBackwards() {is=title AnimGraphNode_StateMachine, bAllowConduitEntryStates=true, bSkipFirstUpdateTransition=false, maxTransitionsPerFrame=1}
   Output Animation Pose(Result=@AnimGraphNode_StateMachine_0)
  ]
 }
 StopBackwards {
  entry=Conduit
  path=VehicleFullBodyOverride.AnimGraphNode_StateMachine_0.FullBody.AnimStateNode_1.Stop.AnimGraphNode_StateMachine_0.StopTransition.AnimStateNode_3.Backwards.AnimGraphNode_StateMachine_0.StopBackwards
  ids=id
  nodes [
   AnimStateConduitNode_0 = Conduit()
   AnimStateNode_0 = LeftBackFeetForward() {is=title AnimStateNode}
   AnimStateNode_1 = RightBackFeetForward() {is=title AnimStateNode}
   AnimStateNode_2 = LeftPlantedRightPass() {is=title AnimStateNode}
   AnimStateNode_3 = RightPlantedLeftPass() {is=title AnimStateNode}
   AnimStateTransitionNode_0 = AnimStateTransitionNode() {logicType=TLT_Inertialization}
   AnimStateTransitionNode_1 = AnimStateTransitionNode() {logicType=TLT_Inertialization, priorityOrder=2}
   AnimStateTransitionNode_2 = AnimStateTransitionNode() {logicType=TLT_Inertialization, priorityOrder=3}
   AnimStateTransitionNode_3 = AnimStateTransitionNode() {logicType=TLT_Inertialization, priorityOrder=100}
  ]
  links [
   {
    dst=LeftBackFeetForward
    rule=VehicleFullBodyOverride.AnimGraphNode_StateMachine_0.FullBody.AnimStateNode_1.Stop.AnimGraphNode_StateMachine_0.StopTransition.AnimStateNode_3.Backwards.AnimGraphNode_StateMachine_0.StopBackwards.AnimStateTransitionNode_0.AnimationTransitionGraph_1
    src=Conduit
    via=AnimStateTransitionNode_0
   }
   {
    dst=LeftPlantedRightPass
    rule=VehicleFullBodyOverride.AnimGraphNode_StateMachine_0.FullBody.AnimStateNode_1.Stop.AnimGraphNode_StateMachine_0.StopTransition.AnimStateNode_3.Backwards.AnimGraphNode_StateMachine_0.StopBackwards.AnimStateTransitionNode_2.AnimationTransitionGraph_1
    src=Conduit
    via=AnimStateTransitionNode_2
   }
   {
    dst=RightBackFeetForward
    rule=VehicleFullBodyOverride.AnimGraphNode_StateMachine_0.FullBody.AnimStateNode_1.Stop.AnimGraphNode_StateMachine_0.StopTransition.AnimStateNode_3.Backwards.AnimGraphNode_StateMachine_0.StopBackwards.AnimStateTransitionNode_1.AnimationTransitionGraph_1
    src=Conduit
    via=AnimStateTransitionNode_1
   }
   {
    dst=RightPlantedLeftPass
    rule=VehicleFullBodyOverride.AnimGraphNode_StateMachine_0.FullBody.AnimStateNode_1.Stop.AnimGraphNode_StateMachine_0.StopTransition.AnimStateNode_3.Backwards.AnimGraphNode_StateMachine_0.StopBackwards.AnimStateTransitionNode_3.AnimationTransitionGraph_1
    src=Conduit
    via=AnimStateTransitionNode_3
   }
  ]
 }
 Conduit {
  path=VehicleFullBodyOverride.AnimGraphNode_StateMachine_0.FullBody.AnimStateNode_1.Stop.AnimGraphNode_StateMachine_0.StopTransition.AnimStateNode_3.Backwards.AnimGraphNode_StateMachine_0.StopBackwards.AnimStateConduitNode_0.Conduit
  nodes [Result(bCanEnterTransition=true)]
 }
 LeftBackFeetForward {
  path=VehicleFullBodyOverride.AnimGraphNode_StateMachine_0.FullBody.AnimStateNode_1.Stop.AnimGraphNode_StateMachine_0.StopTransition.AnimStateNode_3.Backwards.AnimGraphNode_StateMachine_0.StopBackwards.AnimStateNode_0.LeftBackFeetForward
  nodes [Output Animation Pose(Result=Sequence Player 'HardSaddleWolf_Player_Stop_Backwards'() {bLoopAnimation=false, sequence=/RidingContent/Animation/HardSaddle/Wolf/HardSaddleWolf_Player_Stop_Backwards})]
 }
 RightBackFeetForward {
  path=VehicleFullBodyOverride.AnimGraphNode_StateMachine_0.FullBody.AnimStateNode_1.Stop.AnimGraphNode_StateMachine_0.StopTransition.AnimStateNode_3.Backwards.AnimGraphNode_StateMachine_0.StopBackwards.AnimStateNode_1.RightBackFeetForward
  nodes [Output Animation Pose(Result=Sequence Player 'HardSaddleWolf_Player_Stop_Backwards_fr19'() {bLoopAnimation=false, sequence=/RidingContent/Animation/HardSaddle/Wolf/HardSaddleWolf_Player_Stop_Backwards_fr19})]
 }
 LeftPlantedRightPass {
  path=VehicleFullBodyOverride.AnimGraphNode_StateMachine_0.FullBody.AnimStateNode_1.Stop.AnimGraphNode_StateMachine_0.StopTransition.AnimStateNode_3.Backwards.AnimGraphNode_StateMachine_0.StopBackwards.AnimStateNode_2.LeftPlantedRightPass
  nodes [Output Animation Pose(Result=Sequence Player 'HardSaddleWolf_Player_Stop_Backwards_fr24'() {bLoopAnimation=false, sequence=/RidingContent/Animation/HardSaddle/Wolf/HardSaddleWolf_Player_Stop_Backwards_fr24})]
 }
 RightPlantedLeftPass {
  path=VehicleFullBodyOverride.AnimGraphNode_StateMachine_0.FullBody.AnimStateNode_1.Stop.AnimGraphNode_StateMachine_0.StopTransition.AnimStateNode_3.Backwards.AnimGraphNode_StateMachine_0.StopBackwards.AnimStateNode_3.RightPlantedLeftPass
  nodes [Output Animation Pose(Result=Sequence Player 'HardSaddleWolf_Player_Stop_Backwards_fr7'() {bLoopAnimation=false, sequence=/RidingContent/Animation/HardSaddle/Wolf/HardSaddleWolf_Player_Stop_Backwards_fr7})]
 }
 AnimationTransitionGraph_1 {
  path=VehicleFullBodyOverride.AnimGraphNode_StateMachine_0.FullBody.AnimStateNode_1.Stop.AnimGraphNode_StateMachine_0.StopTransition.AnimStateNode_3.Backwards.AnimGraphNode_StateMachine_0.StopBackwards.AnimStateTransitionNode_0.AnimationTransitionGraph_1
  nodes [Result()]
 }
 AnimationTransitionGraph_1 {
  path=VehicleFullBodyOverride.AnimGraphNode_StateMachine_0.FullBody.AnimStateNode_1.Stop.AnimGraphNode_StateMachine_0.StopTransition.AnimStateNode_3.Backwards.AnimGraphNode_StateMachine_0.StopBackwards.AnimStateTransitionNode_1.AnimationTransitionGraph_1
  nodes [Result()]
 }
 AnimationTransitionGraph_1 {
  path=VehicleFullBodyOverride.AnimGraphNode_StateMachine_0.FullBody.AnimStateNode_1.Stop.AnimGraphNode_StateMachine_0.StopTransition.AnimStateNode_3.Backwards.AnimGraphNode_StateMachine_0.StopBackwards.AnimStateTransitionNode_2.AnimationTransitionGraph_1
  nodes [Result()]
 }
 AnimationTransitionGraph_1 {
  path=VehicleFullBodyOverride.AnimGraphNode_StateMachine_0.FullBody.AnimStateNode_1.Stop.AnimGraphNode_StateMachine_0.StopTransition.AnimStateNode_3.Backwards.AnimGraphNode_StateMachine_0.StopBackwards.AnimStateTransitionNode_3.AnimationTransitionGraph_1
  nodes [Result(bCanEnterTransition=true)]
 }
 Transition {
  path=VehicleFullBodyOverride.AnimGraphNode_StateMachine_0.FullBody.AnimStateNode_1.Stop.AnimGraphNode_StateMachine_0.StopTransition.AnimStateTransitionNode_0.Transition
  nodes [Result(bCanEnterTransition=Get Should Play Stop Backwards() -> (Should Play Stop Backwards:bool))]
 }
 Transition {
  path=VehicleFullBodyOverride.AnimGraphNode_StateMachine_0.FullBody.AnimStateNode_1.Stop.AnimGraphNode_StateMachine_0.StopTransition.AnimStateTransitionNode_3.Transition
  nodes [Result()]
 }
 Transition {
  path=VehicleFullBodyOverride.AnimGraphNode_StateMachine_0.FullBody.AnimStateNode_1.Stop.AnimGraphNode_StateMachine_0.StopTransition.AnimStateTransitionNode_4.Transition
  nodes [Result()]
 }
 Transition {
  path=VehicleFullBodyOverride.AnimGraphNode_StateMachine_0.FullBody.AnimStateNode_1.Stop.AnimGraphNode_StateMachine_0.StopTransition.AnimStateTransitionNode_5.Transition
  nodes [Result(bCanEnterTransition=true)]
 }
 Startle {
  path=VehicleFullBodyOverride.AnimGraphNode_StateMachine_0.FullBody.AnimStateNode_10.Startle
  nodes [Output Animation Pose(Result=Sequence Player 'Labrador_Player_Startle'() {bLoopAnimation=false, sequence=&/Labrador_Player_Startle})]
 }
 Idle {
  path=VehicleFullBodyOverride.AnimGraphNode_StateMachine_0.FullBody.AnimStateNode_2.Idle
  nodes [Output Animation Pose(Result=Sequence Player 'Labrador_Player_Idle'() {sequence=&/Labrador_Player_Idle})]
 }
 Moving {
  path=VehicleFullBodyOverride.AnimGraphNode_StateMachine_0.FullBody.AnimStateNode_3.Moving
  nodes [
   n1 = "Sequence Evaluator (by time)"!() -> (Pose:Pose Link Structure) {
    is=title AnimGraphNode_SequenceEvaluator
    becomeRelevantFunction {bSelfContext=true, memberGuid=#9, memberName=OnRelevant_Movement}
    node {becomeRelevantFunction {functionName=OnInitialUpdate_Movement}, updateFunction {functionName=OnUpdate}}
    updateFunction {bSelfContext=true, memberGuid=#10, memberName=OnUpdate_Movement}
   }
   Output Animation Pose(Result=@n1)
  ]
 }
 JumpFall {
  path=VehicleFullBodyOverride.AnimGraphNode_StateMachine_0.FullBody.AnimStateNode_4.JumpFall
  nodes [
   n1 = Blendspace Player 'HardSaddleWolf_Player_SlopeSliding_BS'(X=Get SlopeSlidingRoll() -> (SlopeSlidingRoll:float), Y=Get SlopeSlidingPitch() -> (SlopeSlidingPitch:float)) {blendSpace=/RidingContent/Animation/HardSaddle/Wolf/SlopeSliding/HardSaddleWolf_Player_SlopeSliding_BS}
   n2 = "Sequence Player (additive) 'HardSaddleWolf_Player_SlopeSliding_Add'"() {sequence=/RidingContent/Animation/HardSaddle/Wolf/SlopeSliding/HardSaddleWolf_Player_SlopeSliding_Add}
   Output Animation Pose(Result=Blend Poses by bool(
     BlendPose_0=Apply Additive!(Base=@n1, Additive=@n2, Alpha:float) -> (Pose) {alphaBoolBlend {blendOption=Linear}}
     BlendPose_1=Sequence Player 'Labrador_Player_Jump_FallLoop'() {sequence=&/Labrador_Player_Jump_FallLoop}
     bActiveValue=Get bIsSlopeSliding() -> (bIsSlopeSliding:bool)
    ) {blendType=Linear})
  ]
 }
 Turn {
  path=VehicleFullBodyOverride.AnimGraphNode_StateMachine_0.FullBody.AnimStateNode_5.Turn
  nodes [Output Animation Pose(Result=Blend Poses by bool(
     BlendPose_0=Sequence Player 'HardSaddleWolf_Player_TurnRight'!(PlayRate:float=1) -> (Pose) {sequence=/RidingContent/Animation/HardSaddle/Wolf/HardSaddleWolf_Player_TurnRight}
     BlendPose_1=Sequence Player 'HardSaddleWolf_Player_TurnLeft'!(PlayRate:float=1) -> (Pose) {sequence=/RidingContent/Animation/HardSaddle/Wolf/HardSaddleWolf_Player_TurnLeft}
    ) {blendType=Linear})]
 }
 JumpLand {
  path=VehicleFullBodyOverride.AnimGraphNode_StateMachine_0.FullBody.AnimStateNode_6.JumpLand
  nodes [Output Animation Pose(Result=Sequence Player 'Labrador_Player_Jump_Land'() {bLoopAnimation=false, sequence=&/Labrador_Player_Jump_Land})]
 }
 Swimming {
  path=VehicleFullBodyOverride.AnimGraphNode_StateMachine_0.FullBody.AnimStateNode_7.Swimming
  nodes [
   n1 = Sequence Player 'HardSaddleWolf_Player_Swim_N'() {groupName=Locomotion, method=SyncGroup, sequence=/RidingContent/Animation/HardSaddle/Wolf/HardSaddleWolf_Player_Swim_N}
   n2 = Sequence Player 'HardSaddleWolf_Player_Swim_S'() {groupName=Locomotion, method=SyncGroup, sequence=/RidingContent/Animation/HardSaddle/Wolf/HardSaddleWolf_Player_Swim_S}
   n3 = Get Is Moving Backwards() -> (Is Moving Backwards:bool)
   Output Animation Pose(Result=Blend Poses by bool(
     BlendPose_0=Blend Poses by bool(BlendPose_0=@n2, BlendPose_1=@n1, BlendTime_0=0.35, BlendTime_1=0.35, bActiveValue=@n3) {blendType=Linear, transitionType=Inertialization}
     BlendPose_1=Sequence Player 'HardSaddleWolf_Player_Idle_Swim'() {sequence=/RidingContent/Animation/HardSaddle/Wolf/HardSaddleWolf_Player_Idle_Swim}
     BlendTime_0=0.5
     BlendTime_1=1
     bActiveValue=Get State Rule Swim Idle Swim Loco() -> (State Rule Swim Idle Swim Loco:bool)
    ) {blendType=Linear})
  ]
 }
 Transition {
  path=VehicleFullBodyOverride.AnimGraphNode_StateMachine_0.FullBody.AnimStateTransitionNode_0.Transition
  nodes [Result(bCanEnterTransition=Get Is Jumping())]
 }
 Transition {
  path=VehicleFullBodyOverride.AnimGraphNode_StateMachine_0.FullBody.AnimStateTransitionNode_1.Transition
  nodes [Result(bCanEnterTransition=Get Jump Start To Jump Loop() -> (Jump Start To Jump Loop:bool))]
 }
 Transition {
  path=VehicleFullBodyOverride.AnimGraphNode_StateMachine_0.FullBody.AnimStateTransitionNode_10.Transition
  nodes [Result(bCanEnterTransition=Get State Rule Jump Land Moving() -> (State Rule Jump Land Moving:bool))]
 }
 Transition {
  path=VehicleFullBodyOverride.AnimGraphNode_StateMachine_0.FullBody.AnimStateTransitionNode_11.Transition
  nodes [Result(bCanEnterTransition=Get State Rule Idle Moving() -> (State Rule Idle Moving:bool))]
 }
 Transition {
  path=VehicleFullBodyOverride.AnimGraphNode_StateMachine_0.FullBody.AnimStateTransitionNode_12.Transition
  nodes [Result(bCanEnterTransition=Get Is Turning in Place() -> (Is Turning in Place:bool))]
 }
 Transition {
  path=VehicleFullBodyOverride.AnimGraphNode_StateMachine_0.FullBody.AnimStateTransitionNode_13.Transition
  nodes [Result(bCanEnterTransition=Get Is Falling())]
 }
 Transition {
  path=VehicleFullBodyOverride.AnimGraphNode_StateMachine_0.FullBody.AnimStateTransitionNode_14.Transition
  nodes [
   n1 = Get bIsJumping!() -> (bIsJumping)
   Result(bCanEnterTransition=AND Boolean!(A=Get State Rule Idle Moving(), B=NOT Boolean(A=@n1), C:bool=Get JustLanded() -> (JustLanded:bool)) -> (ReturnValue))
  ]
 }
 Transition {
  path=VehicleFullBodyOverride.AnimGraphNode_StateMachine_0.FullBody.AnimStateTransitionNode_15.Transition
  nodes [Result(bCanEnterTransition=Get State Rule Idle Moving())]
 }
 Transition {
  path=VehicleFullBodyOverride.AnimGraphNode_StateMachine_0.FullBody.AnimStateTransitionNode_18.Transition
  nodes [Result(bCanEnterTransition=Get Loco To Idle() -> (Loco To Idle:bool))]
 }
 Transition {
  path=VehicleFullBodyOverride.AnimGraphNode_StateMachine_0.FullBody.AnimStateTransitionNode_2.Transition
  nodes [Result(bCanEnterTransition=Get State Rule Moving Idle() -> (State Rule Moving Idle:bool))]
 }
 AnimationTransitionGraph_0 {
  path=VehicleFullBodyOverride.AnimGraphNode_StateMachine_0.FullBody.AnimStateTransitionNode_23.AnimationTransitionGraph_0
  nodes [
   n1 = Get JustLanded()
   n2 = Get bIsJumping()
   Result(bCanEnterTransition=AND Boolean(A=Get State Rule Idle Moving(), B=NOT Boolean(A=@n2), C=NOT Boolean(A=@n1)))
  ]
 }
 AnimationTransitionGraph_0 {path=VehicleFullBodyOverride.AnimGraphNode_StateMachine_0.FullBody.AnimStateTransitionNode_24.AnimationTransitionGraph_0, nodes [Result()]}
 Transition {path=VehicleFullBodyOverride.AnimGraphNode_StateMachine_0.FullBody.AnimStateTransitionNode_3.Transition, nodes [Result()]}
 Transition {
  path=VehicleFullBodyOverride.AnimGraphNode_StateMachine_0.FullBody.AnimStateTransitionNode_4.Transition
  nodes [Result(bCanEnterTransition=Get State Rule Fall Idle() -> (State Rule Fall Idle:bool))]
 }
 Transition {path=VehicleFullBodyOverride.AnimGraphNode_StateMachine_0.FullBody.AnimStateTransitionNode_5.Transition, nodes [Result()]}
 Transition {
  path=VehicleFullBodyOverride.AnimGraphNode_StateMachine_0.FullBody.AnimStateTransitionNode_6.Transition
  nodes [Result(bCanEnterTransition=Get Is Surface Swimming() -> (Is Surface Swimming:bool))]
 }
 Transition {
  path=VehicleFullBodyOverride.AnimGraphNode_StateMachine_0.FullBody.AnimStateTransitionNode_7.Transition
  nodes [Result(bCanEnterTransition=NOT Boolean(A=Get Is Surface Swimming()))]
 }
 Transition {
  path=VehicleFullBodyOverride.AnimGraphNode_StateMachine_0.FullBody.AnimStateTransitionNode_8.Transition
  nodes [Result(bCanEnterTransition=Get Is Turning in Place())]
 }
 Transition {
  path=VehicleFullBodyOverride.AnimGraphNode_StateMachine_0.FullBody.AnimStateTransitionNode_9.Transition
  nodes [Result(bCanEnterTransition=Get State Rule Fall Moving() -> (State Rule Fall Moving:bool))]
 }
 UpperBody {
  entry=FullBodyPose
  path=VehicleFullBodyOverride.AnimGraphNode_StateMachine_1.UpperBody
  ids=id
  nodes [
   AnimStateNode_0 = HackFullBodyPose() {is=title AnimStateNode}
   AnimStateNode_1 = CombatToRide() {is=title AnimStateNode}
   AnimStateNode_2 = FullBodyPose() {is=title AnimStateNode}
   AnimStateTransitionNode_5 = AnimStateTransitionNode() {bDisabled=true, crossfadeDuration=0.1, logicType=TLT_Inertialization}
   AnimStateTransitionNode_6 = AnimStateTransitionNode() {blendMode=Linear, crossfadeDuration=0, logicType=TLT_Inertialization}
   AnimStateTransitionNode_8 = AnimStateTransitionNode() {bAutomaticRuleBasedOnSequencePlayerInState=true, logicType=TLT_Inertialization}
  ]
  links [
   {
    dst=HackFullBodyPose
    rule=VehicleFullBodyOverride.AnimGraphNode_StateMachine_1.UpperBody.AnimStateTransitionNode_8.Transition
    src=CombatToRide
    via=AnimStateTransitionNode_8
   }
   {
    dst=CombatToRide
    rule=VehicleFullBodyOverride.AnimGraphNode_StateMachine_1.UpperBody.AnimStateTransitionNode_5.Transition
    src=FullBodyPose
    via=AnimStateTransitionNode_5
   }
   {
    dst=FullBodyPose
    rule=VehicleFullBodyOverride.AnimGraphNode_StateMachine_1.UpperBody.AnimStateTransitionNode_6.Transition
    src=HackFullBodyPose
    via=AnimStateTransitionNode_6
   }
  ]
 }
 HackFullBodyPose {
  path=VehicleFullBodyOverride.AnimGraphNode_StateMachine_1.UpperBody.AnimStateNode_0.HackFullBodyPose
  nodes [Output Animation Pose(Result=Use cached pose 'FullBodyPose'())]
 }
 CombatToRide {
  path=VehicleFullBodyOverride.AnimGraphNode_StateMachine_1.UpperBody.AnimStateNode_1.CombatToRide
  nodes [Output Animation Pose(Result=Blendspace Player 'HardSaddleWolf_Player_CtR_BS'!(X=Get CombatToRideYaw() -> (CombatToRideYaw:float)) -> (Pose) {bLoop=false, blendSpace=/RidingContent/Animation/HardSaddle/Wolf/HardSaddleWolf_Player_CtR_BS})]
 }
 FullBodyPose {
  path=VehicleFullBodyOverride.AnimGraphNode_StateMachine_1.UpperBody.AnimStateNode_2.FullBodyPose
  nodes [Output Animation Pose(Result=Use cached pose 'FullBodyPose'())]
 }
 Transition {
  path=VehicleFullBodyOverride.AnimGraphNode_StateMachine_1.UpperBody.AnimStateTransitionNode_5.Transition
  nodes [Result(bCanEnterTransition=Get bIsRideMode())]
 }
 Transition {
  path=VehicleFullBodyOverride.AnimGraphNode_StateMachine_1.UpperBody.AnimStateTransitionNode_6.Transition
  nodes [Result(bCanEnterTransition=NOT Boolean(A=Get bIsRideMode()))]
 }
 Transition {path=VehicleFullBodyOverride.AnimGraphNode_StateMachine_1.UpperBody.AnimStateTransitionNode_8.Transition, nodes [Result()]}
 CombatUpperBodyAdditive {
  entry=Idle
  path=VehicleFullBodyOverride.AnimGraphNode_StateMachine_2.CombatUpperBodyAdditive
  ids=id
  nodes [
   SA_ToSwimming() {bGlobalAlias=true}
   SA_ToJumpStart() {also {aliases [Idle,JumpLand,Moving,Stop,Turn,TurnSettle]}}
   SA_ToJumpFall() {also {aliases [JumpLand,Moving,Stop,Turn,TurnSettle]}}
   SA_ToMoving() {also {aliases [Idle,JumpLand,Stop,Turn,TurnSettle]}}
   AnimStateNode_0 = Moving() {
    stateEntered {guid=#11}
    stateFullyBlended {guid=#12}
    stateLeft {guid=#13}
   }
   AnimStateNode_1 = Turn()
   AnimStateNode_2 = Swimming()
   AnimStateNode_4 = Idle()
   AnimStateNode_5 = JumpStart()
   AnimStateNode_6 = JumpFall()
   AnimStateNode_7 = Stop()
   AnimStateNode_8 = JumpLand()
   AnimStateNode_9 = TurnSettle() {is=title AnimStateNode}
   AnimStateTransitionNode_0 = AnimStateTransitionNode()
   AnimStateTransitionNode_1 = AnimStateTransitionNode() {bAutomaticRuleBasedOnSequencePlayerInState=true, crossfadeDuration=0.3, logicType=TLT_Inertialization, priorityOrder=100}
   AnimStateTransitionNode_10 = AnimStateTransitionNode() {crossfadeDuration=0.35, logicType=TLT_Inertialization, priorityOrder=3}
   AnimStateTransitionNode_11 = AnimStateTransitionNode() {crossfadeDuration=0.3, logicType=TLT_Inertialization}
   AnimStateTransitionNode_12 = AnimStateTransitionNode() {blendMode=Linear, crossfadeDuration=0.3, priorityOrder=2}
   AnimStateTransitionNode_13 = AnimStateTransitionNode() {crossfadeDuration=0.4}
   AnimStateTransitionNode_14 = AnimStateTransitionNode() {logicType=TLT_Inertialization}
   AnimStateTransitionNode_15 = AnimStateTransitionNode() {bAutomaticRuleBasedOnSequencePlayerInState=true, crossfadeDuration=0.3, priorityOrder=2}
   AnimStateTransitionNode_16 = AnimStateTransitionNode()
   AnimStateTransitionNode_18 = AnimStateTransitionNode() {bAutomaticRuleBasedOnSequencePlayerInState=true, crossfadeDuration=0.4, priorityOrder=3}
   AnimStateTransitionNode_2 = AnimStateTransitionNode() {
    crossfadeDuration=0.6
    logicType=TLT_Inertialization
    priorityOrder=2
    transitionEnd {guid=#14}
    transitionInterrupt {guid=#15}
    transitionStart {guid=#16}
   }
   AnimStateTransitionNode_3 = AnimStateTransitionNode() {crossfadeDuration=0.4, logicType=TLT_Inertialization}
   AnimStateTransitionNode_4 = AnimStateTransitionNode() {crossfadeDuration=0.3, logicType=TLT_Inertialization, priorityOrder=100}
   AnimStateTransitionNode_5 = AnimStateTransitionNode() {logicType=TLT_Inertialization}
   AnimStateTransitionNode_6 = AnimStateTransitionNode() {logicType=TLT_Inertialization, priorityOrder=100}
   AnimStateTransitionNode_7 = AnimStateTransitionNode() {bAutomaticRuleBasedOnSequencePlayerInState=true, crossfadeDuration=0.4, logicType=TLT_Inertialization}
   AnimStateTransitionNode_8 = AnimStateTransitionNode() {crossfadeDuration=0.15, logicType=TLT_Inertialization}
   AnimStateTransitionNode_9 = AnimStateTransitionNode() {logicType=TLT_Inertialization, priorityOrder=2}
  ]
  links [
   {
    dst=Turn
    rule=VehicleFullBodyOverride.AnimGraphNode_StateMachine_2.CombatUpperBodyAdditive.AnimStateTransitionNode_0.Transition
    src=Idle
    via=AnimStateTransitionNode_0
   }
   {
    dst=JumpLand
    rule=VehicleFullBodyOverride.AnimGraphNode_StateMachine_2.CombatUpperBodyAdditive.AnimStateTransitionNode_6.Transition
    src=JumpFall
    via=AnimStateTransitionNode_6
   }
   {
    dst=Moving
    rule=VehicleFullBodyOverride.AnimGraphNode_StateMachine_2.CombatUpperBodyAdditive.AnimStateTransitionNode_8.Transition
    src=JumpFall
    via=AnimStateTransitionNode_8
   }
   {
    dst=Idle
    rule=VehicleFullBodyOverride.AnimGraphNode_StateMachine_2.CombatUpperBodyAdditive.AnimStateTransitionNode_1.Transition
    src=JumpLand
    via=AnimStateTransitionNode_1
   }
   {
    dst=Moving
    rule=VehicleFullBodyOverride.AnimGraphNode_StateMachine_2.CombatUpperBodyAdditive.AnimStateTransitionNode_9.Transition
    src=JumpLand
    via=AnimStateTransitionNode_9
   }
   {
    dst=JumpFall
    rule=VehicleFullBodyOverride.AnimGraphNode_StateMachine_2.CombatUpperBodyAdditive.AnimStateTransitionNode_7.Transition
    src=JumpStart
    via=AnimStateTransitionNode_7
   }
   {
    dst=Stop
    rule=VehicleFullBodyOverride.AnimGraphNode_StateMachine_2.CombatUpperBodyAdditive.AnimStateTransitionNode_10.Transition
    src=Moving
    via=AnimStateTransitionNode_10
   }
   {
    dst=JumpFall
    rule=VehicleFullBodyOverride.AnimGraphNode_StateMachine_2.CombatUpperBodyAdditive.AnimStateTransitionNode_11.Transition
    src=SA_ToJumpFall
    via=AnimStateTransitionNode_11
   }
   {
    dst=JumpStart
    rule=VehicleFullBodyOverride.AnimGraphNode_StateMachine_2.CombatUpperBodyAdditive.AnimStateTransitionNode_4.Transition
    src=SA_ToJumpStart
    via=AnimStateTransitionNode_4
   }
   {
    dst=Moving
    rule=VehicleFullBodyOverride.AnimGraphNode_StateMachine_2.CombatUpperBodyAdditive.AnimStateTransitionNode_2.Transition
    src=SA_ToMoving
    via=AnimStateTransitionNode_2
   }
   {
    dst=Swimming
    rule=VehicleFullBodyOverride.AnimGraphNode_StateMachine_2.CombatUpperBodyAdditive.AnimStateTransitionNode_3.Transition
    src=SA_ToSwimming
    via=AnimStateTransitionNode_3
   }
   {
    dst=Idle
    rule=VehicleFullBodyOverride.AnimGraphNode_StateMachine_2.CombatUpperBodyAdditive.AnimStateTransitionNode_18.Transition
    src=Stop
    via=AnimStateTransitionNode_18
   }
   {
    dst=Moving
    rule=VehicleFullBodyOverride.AnimGraphNode_StateMachine_2.CombatUpperBodyAdditive.AnimStateTransitionNode_12.Transition
    src=Stop
    via=AnimStateTransitionNode_12
   }
   {
    dst=Turn
    rule=VehicleFullBodyOverride.AnimGraphNode_StateMachine_2.CombatUpperBodyAdditive.AnimStateTransitionNode_13.Transition
    src=Stop
    via=AnimStateTransitionNode_13
   }
   {
    dst=Moving
    rule=VehicleFullBodyOverride.AnimGraphNode_StateMachine_2.CombatUpperBodyAdditive.AnimStateTransitionNode_5.Transition
    src=Swimming
    via=AnimStateTransitionNode_5
   }
   {
    dst=TurnSettle
    rule=VehicleFullBodyOverride.AnimGraphNode_StateMachine_2.CombatUpperBodyAdditive.AnimStateTransitionNode_14.Transition
    src=Turn
    via=AnimStateTransitionNode_14
   }
   {
    dst=Idle
    rule=VehicleFullBodyOverride.AnimGraphNode_StateMachine_2.CombatUpperBodyAdditive.AnimStateTransitionNode_15.Transition
    src=TurnSettle
    via=AnimStateTransitionNode_15
   }
   {
    dst=Turn
    rule=VehicleFullBodyOverride.AnimGraphNode_StateMachine_2.CombatUpperBodyAdditive.AnimStateTransitionNode_16.Transition
    src=TurnSettle
    via=AnimStateTransitionNode_16
   }
  ]
 }
 Moving {
  path=VehicleFullBodyOverride.AnimGraphNode_StateMachine_2.CombatUpperBodyAdditive.AnimStateNode_0.Moving
  nodes [
   n1 = Get CurrentCurveValue()
   Output Animation Pose(Result=Blend Poses by bool(
     BlendPose_0="Sequence Evaluator (by time) 'Labrador_Player_Run_N_UpperBodyAdditive'"!(ExplicitTime:float=@n1) -> (Pose:Pose Link Structure) {sequence=&/Labrador_Player_Run_N_UpperBodyAdditive}
     BlendPose_1="Sequence Evaluator (by time) 'Labrador_Player_Walk_N_UpperBodyAdditive'"!(ExplicitTime:float=@n1) -> (Pose:Pose Link Structure) {sequence=&/Labrador_Player_Walk_N_UpperBodyAdditive}
     BlendTime_0=0.5
     BlendTime_1=0.7
     bActiveValue=Get IsRunning() -> (IsRunning:bool)
    ) {blendType=Linear, transitionType=Inertialization})
  ]
 }
 Turn {
  path=VehicleFullBodyOverride.AnimGraphNode_StateMachine_2.CombatUpperBodyAdditive.AnimStateNode_1.Turn
  nodes [
   n1 = float / float(A=Get Turn in Place Rotation Speed() -> (Turn in Place Rotation Speed:double), B=168.75)
   n2 = Get Turn in Place Angle Delta() -> (Turn in Place Angle Delta:double)
   Output Animation Pose(Result=Blend Poses by bool(
     BlendPose_0="Sequence Player (additive) 'HardSaddleWolf_Player_TurnRight_Combat'"!(PlayRate:float=@n1) -> (Pose) {bLoopAnimation=false, sequence=/RidingContent/Animation/HardSaddle/Wolf/HardSaddleWolf_Player_TurnRight_Combat}
     BlendPose_1="Sequence Player (additive) 'HardSaddleWolf_Player_TurnLeft_Combat'"!(PlayRate:float=@n1) -> (Pose) {bLoopAnimation=false, sequence=/RidingContent/Animation/HardSaddle/Wolf/HardSaddleWolf_Player_TurnLeft_Combat}
     bActiveValue="float > float"(A=@n2)
    ) {blendType=Linear})
  ]
 }
 Swimming {
  path=VehicleFullBodyOverride.AnimGraphNode_StateMachine_2.CombatUpperBodyAdditive.AnimStateNode_2.Swimming
  nodes [
   n1 = "Sequence Player (additive) 'HardSaddle_Player_Combat_Swim_N_Additive'"() {groupName=Locomotion, method=SyncGroup, sequence=/RidingContent/Animation/HardSaddle/Wolf/HardSaddle_Player_Combat_Swim_N_Additive}
   n2 = "Sequence Player (additive) 'HardSaddle_Player_Combat_Swim_S_Additive'"() {groupName=Locomotion, method=SyncGroup, sequence=/RidingContent/Animation/HardSaddle/Wolf/HardSaddle_Player_Combat_Swim_S_Additive}
   n3 = Get Is Moving Backwards()
   Output Animation Pose(Result=Blend Poses by bool(
     BlendPose_0=Blend Poses by bool(BlendPose_0=@n2, BlendPose_1=@n1, BlendTime_0=0.35, BlendTime_1=0.35, bActiveValue=@n3) {blendType=Linear, transitionType=Inertialization}
     BlendPose_1="Sequence Player (additive) 'HardSaddle_Player_Combat_IdleSwim_Additive'"() {sequence=/RidingContent/Animation/HardSaddle/Wolf/HardSaddle_Player_Combat_IdleSwim_Additive}
     BlendTime_0=0.5
     BlendTime_1=1
     bActiveValue=Get State Rule Swim Idle Swim Loco()
    ) {blendType=Linear})
  ]
 }
 Idle {path=VehicleFullBodyOverride.AnimGraphNode_StateMachine_2.CombatUpperBodyAdditive.AnimStateNode_4.Idle, nodes [Output Animation Pose()]}
 JumpStart {
  path=VehicleFullBodyOverride.AnimGraphNode_StateMachine_2.CombatUpperBodyAdditive.AnimStateNode_5.JumpStart
  nodes [Output Animation Pose(Result="Sequence Player (additive) 'HardSaddleWolf_Player_CombatJumpStart_Add'"() {bLoopAnimation=false, sequence=/RidingContent/Animation/HardSaddle/Wolf/HardSaddleWolf_Player_CombatJumpStart_Add})]
 }
 JumpFall {
  path=VehicleFullBodyOverride.AnimGraphNode_StateMachine_2.CombatUpperBodyAdditive.AnimStateNode_6.JumpFall
  nodes [Output Animation Pose(Result=Blend Poses by bool(
     BlendPose_0="Sequence Player (additive) 'HardSaddleWolf_plyr_SlopeSliding_AddUpper'"() {sequence=/RidingContent/Animation/HardSaddle/Wolf/SlopeSliding/HardSaddleWolf_plyr_SlopeSliding_AddUpper}
     BlendPose_1="Sequence Player (additive) 'HardSaddleWolf_Player_CombatJumpLoop_Add'"() {sequence=/RidingContent/Animation/HardSaddle/Wolf/HardSaddleWolf_Player_CombatJumpLoop_Add}
     bActiveValue=Get bIsSlopeSliding()
    ) {blendType=Linear})]
 }
 Stop {
  path=VehicleFullBodyOverride.AnimGraphNode_StateMachine_2.CombatUpperBodyAdditive.AnimStateNode_7.Stop
  nodes [
   AnimGraphNode_StateMachine_0 = AdditiveStopTransition() {is=title AnimGraphNode_StateMachine, bAllowConduitEntryStates=true, bSkipFirstUpdateTransition=false, maxTransitionsPerFrame=1}
   Output Animation Pose(Result=@AnimGraphNode_StateMachine_0)
  ]
 }
 AdditiveStopTransition {
  entry=Conduit
  path=VehicleFullBodyOverride.AnimGraphNode_StateMachine_2.CombatUpperBodyAdditive.AnimStateNode_7.Stop.AnimGraphNode_StateMachine_0.AdditiveStopTransition
  ids=id
  nodes [
   AnimStateConduitNode_0 = Conduit()
   AnimStateNode_0 = FeetInAir()
   AnimStateNode_1 = FrontFeetPlanted()
   AnimStateNode_2 = BackFeetPlanted()
   AnimStateNode_3 = Backwards()
   AnimStateTransitionNode_1 = AnimStateTransitionNode() {logicType=TLT_Inertialization, priorityOrder=0}
   AnimStateTransitionNode_3 = AnimStateTransitionNode() {logicType=TLT_Inertialization}
   AnimStateTransitionNode_4 = AnimStateTransitionNode() {logicType=TLT_Inertialization, priorityOrder=2}
   AnimStateTransitionNode_5 = AnimStateTransitionNode() {logicType=TLT_Inertialization, priorityOrder=100}
  ]
  links [
   {
    dst=BackFeetPlanted
    rule=VehicleFullBodyOverride.AnimGraphNode_StateMachine_2.CombatUpperBodyAdditive.AnimStateNode_7.Stop.AnimGraphNode_StateMachine_0.AdditiveStopTransition.AnimStateTransitionNode_5.Transition
    src=Conduit
    via=AnimStateTransitionNode_5
   }
   {
    dst=Backwards
    rule=VehicleFullBodyOverride.AnimGraphNode_StateMachine_2.CombatUpperBodyAdditive.AnimStateNode_7.Stop.AnimGraphNode_StateMachine_0.AdditiveStopTransition.AnimStateTransitionNode_1.Transition
    src=Conduit
    via=AnimStateTransitionNode_1
   }
   {
    dst=FeetInAir
    rule=VehicleFullBodyOverride.AnimGraphNode_StateMachine_2.CombatUpperBodyAdditive.AnimStateNode_7.Stop.AnimGraphNode_StateMachine_0.AdditiveStopTransition.AnimStateTransitionNode_3.Transition
    src=Conduit
    via=AnimStateTransitionNode_3
   }
   {
    dst=FrontFeetPlanted
    rule=VehicleFullBodyOverride.AnimGraphNode_StateMachine_2.CombatUpperBodyAdditive.AnimStateNode_7.Stop.AnimGraphNode_StateMachine_0.AdditiveStopTransition.AnimStateTransitionNode_4.Transition
    src=Conduit
    via=AnimStateTransitionNode_4
   }
  ]
 }
 Conduit {
  path=VehicleFullBodyOverride.AnimGraphNode_StateMachine_2.CombatUpperBodyAdditive.AnimStateNode_7.Stop.AnimGraphNode_StateMachine_0.AdditiveStopTransition.AnimStateConduitNode_0.Conduit
  nodes [Result(bCanEnterTransition=true)]
 }
 FeetInAir {
  path=VehicleFullBodyOverride.AnimGraphNode_StateMachine_2.CombatUpperBodyAdditive.AnimStateNode_7.Stop.AnimGraphNode_StateMachine_0.AdditiveStopTransition.AnimStateNode_0.FeetInAir
  nodes [Output Animation Pose(Result="Sequence Player (additive) 'HardSaddleWolf_Player_Stop_FeetInAir_Combat_Additive'"() {bLoopAnimation=false, sequence=/RidingContent/Animation/HardSaddle/Wolf/HardSaddleWolf_Player_Stop_FeetInAir_Combat_Additive})]
 }
 FrontFeetPlanted {
  path=VehicleFullBodyOverride.AnimGraphNode_StateMachine_2.CombatUpperBodyAdditive.AnimStateNode_7.Stop.AnimGraphNode_StateMachine_0.AdditiveStopTransition.AnimStateNode_1.FrontFeetPlanted
  nodes [Output Animation Pose(Result="Sequence Player (additive) 'HardSaddleWolf_Player_Stop_FrontFeetPlanted_Combat_Additive'"() {bLoopAnimation=false, sequence=/RidingContent/Animation/HardSaddle/Wolf/HardSaddleWolf_Player_Stop_FrontFeetPlanted_Combat_Additive})]
 }
 BackFeetPlanted {
  path=VehicleFullBodyOverride.AnimGraphNode_StateMachine_2.CombatUpperBodyAdditive.AnimStateNode_7.Stop.AnimGraphNode_StateMachine_0.AdditiveStopTransition.AnimStateNode_2.BackFeetPlanted
  nodes [Output Animation Pose(Result="Sequence Player (additive) 'HardSaddleWolf_Player_Stop_BackFeetPlanted_Combat_Additive'"() {bLoopAnimation=false, sequence=/RidingContent/Animation/HardSaddle/Wolf/HardSaddleWolf_Player_Stop_BackFeetPlanted_Combat_Additive})]
 }
 Backwards {
  path=VehicleFullBodyOverride.AnimGraphNode_StateMachine_2.CombatUpperBodyAdditive.AnimStateNode_7.Stop.AnimGraphNode_StateMachine_0.AdditiveStopTransition.AnimStateNode_3.Backwards
  nodes [
   AnimGraphNode_StateMachine_0 = AdditiveStopBackwards() {is=title AnimGraphNode_StateMachine, bAllowConduitEntryStates=true, bSkipFirstUpdateTransition=false, maxTransitionsPerFrame=1}
   Output Animation Pose(Result=@AnimGraphNode_StateMachine_0)
  ]
 }
 AdditiveStopBackwards {
  entry=Conduit
  path=VehicleFullBodyOverride.AnimGraphNode_StateMachine_2.CombatUpperBodyAdditive.AnimStateNode_7.Stop.AnimGraphNode_StateMachine_0.AdditiveStopTransition.AnimStateNode_3.Backwards.AnimGraphNode_StateMachine_0.AdditiveStopBackwards
  ids=id
  nodes [
   AnimStateConduitNode_0 = Conduit()
   AnimStateNode_0 = LeftBackFeetForward()
   AnimStateNode_1 = RightBackFeetForward()
   AnimStateNode_2 = LeftPlantedRightPass()
   AnimStateNode_3 = RightPlantedLeftPass()
   AnimStateTransitionNode_0 = AnimStateTransitionNode() {logicType=TLT_Inertialization}
   AnimStateTransitionNode_1 = AnimStateTransitionNode() {logicType=TLT_Inertialization, priorityOrder=2}
   AnimStateTransitionNode_2 = AnimStateTransitionNode() {logicType=TLT_Inertialization, priorityOrder=3}
   AnimStateTransitionNode_3 = AnimStateTransitionNode() {logicType=TLT_Inertialization, priorityOrder=100}
  ]
  links [
   {
    dst=LeftBackFeetForward
    rule=VehicleFullBodyOverride.AnimGraphNode_StateMachine_2.CombatUpperBodyAdditive.AnimStateNode_7.Stop.AnimGraphNode_StateMachine_0.AdditiveStopTransition.AnimStateNode_3.Backwards.AnimGraphNode_StateMachine_0.AdditiveStopBackwards.AnimStateTransitionNode_0.AnimationTransitionGraph_1
    src=Conduit
    via=AnimStateTransitionNode_0
   }
   {
    dst=LeftPlantedRightPass
    rule=VehicleFullBodyOverride.AnimGraphNode_StateMachine_2.CombatUpperBodyAdditive.AnimStateNode_7.Stop.AnimGraphNode_StateMachine_0.AdditiveStopTransition.AnimStateNode_3.Backwards.AnimGraphNode_StateMachine_0.AdditiveStopBackwards.AnimStateTransitionNode_2.AnimationTransitionGraph_1
    src=Conduit
    via=AnimStateTransitionNode_2
   }
   {
    dst=RightBackFeetForward
    rule=VehicleFullBodyOverride.AnimGraphNode_StateMachine_2.CombatUpperBodyAdditive.AnimStateNode_7.Stop.AnimGraphNode_StateMachine_0.AdditiveStopTransition.AnimStateNode_3.Backwards.AnimGraphNode_StateMachine_0.AdditiveStopBackwards.AnimStateTransitionNode_1.AnimationTransitionGraph_1
    src=Conduit
    via=AnimStateTransitionNode_1
   }
   {
    dst=RightPlantedLeftPass
    rule=VehicleFullBodyOverride.AnimGraphNode_StateMachine_2.CombatUpperBodyAdditive.AnimStateNode_7.Stop.AnimGraphNode_StateMachine_0.AdditiveStopTransition.AnimStateNode_3.Backwards.AnimGraphNode_StateMachine_0.AdditiveStopBackwards.AnimStateTransitionNode_3.AnimationTransitionGraph_1
    src=Conduit
    via=AnimStateTransitionNode_3
   }
  ]
 }
 Conduit {
  path=VehicleFullBodyOverride.AnimGraphNode_StateMachine_2.CombatUpperBodyAdditive.AnimStateNode_7.Stop.AnimGraphNode_StateMachine_0.AdditiveStopTransition.AnimStateNode_3.Backwards.AnimGraphNode_StateMachine_0.AdditiveStopBackwards.AnimStateConduitNode_0.Conduit
  nodes [Result(bCanEnterTransition=true)]
 }
 LeftBackFeetForward {
  path=VehicleFullBodyOverride.AnimGraphNode_StateMachine_2.CombatUpperBodyAdditive.AnimStateNode_7.Stop.AnimGraphNode_StateMachine_0.AdditiveStopTransition.AnimStateNode_3.Backwards.AnimGraphNode_StateMachine_0.AdditiveStopBackwards.AnimStateNode_0.LeftBackFeetForward
  nodes [Output Animation Pose(Result="Sequence Player (additive) 'HardSaddleWolf_Player_Stop_Backwards_Combat_Add'"() {bLoopAnimation=false, sequence=/RidingContent/Animation/HardSaddle/Wolf/HardSaddleWolf_Player_Stop_Backwards_Combat_Add})]
 }
 RightBackFeetForward {
  path=VehicleFullBodyOverride.AnimGraphNode_StateMachine_2.CombatUpperBodyAdditive.AnimStateNode_7.Stop.AnimGraphNode_StateMachine_0.AdditiveStopTransition.AnimStateNode_3.Backwards.AnimGraphNode_StateMachine_0.AdditiveStopBackwards.AnimStateNode_1.RightBackFeetForward
  nodes [Output Animation Pose(Result="Sequence Player (additive) 'HardSaddleWolf_Player_Stop_Backwards_Combat_Add_fr19'"() {sequence=/RidingContent/Animation/HardSaddle/Wolf/HardSaddleWolf_Player_Stop_Backwards_Combat_Add_fr19})]
 }
 LeftPlantedRightPass {
  path=VehicleFullBodyOverride.AnimGraphNode_StateMachine_2.CombatUpperBodyAdditive.AnimStateNode_7.Stop.AnimGraphNode_StateMachine_0.AdditiveStopTransition.AnimStateNode_3.Backwards.AnimGraphNode_StateMachine_0.AdditiveStopBackwards.AnimStateNode_2.LeftPlantedRightPass
  nodes [Output Animation Pose(Result="Sequence Player (additive) 'HardSaddleWolf_Player_Stop_Backwards_Combat_Add_fr24'"() {bLoopAnimation=false, sequence=/RidingContent/Animation/HardSaddle/Wolf/HardSaddleWolf_Player_Stop_Backwards_Combat_Add_fr24})]
 }
 RightPlantedLeftPass {
  path=VehicleFullBodyOverride.AnimGraphNode_StateMachine_2.CombatUpperBodyAdditive.AnimStateNode_7.Stop.AnimGraphNode_StateMachine_0.AdditiveStopTransition.AnimStateNode_3.Backwards.AnimGraphNode_StateMachine_0.AdditiveStopBackwards.AnimStateNode_3.RightPlantedLeftPass
  nodes [Output Animation Pose(Result="Sequence Player (additive) 'HardSaddleWolf_Player_Stop_Backwards_Combat_Add_fr7'"() {bLoopAnimation=false, sequence=/RidingContent/Animation/HardSaddle/Wolf/HardSaddleWolf_Player_Stop_Backwards_Combat_Add_fr7})]
 }
 AnimationTransitionGraph_1 {
  path=VehicleFullBodyOverride.AnimGraphNode_StateMachine_2.CombatUpperBodyAdditive.AnimStateNode_7.Stop.AnimGraphNode_StateMachine_0.AdditiveStopTransition.AnimStateNode_3.Backwards.AnimGraphNode_StateMachine_0.AdditiveStopBackwards.AnimStateTransitionNode_0.AnimationTransitionGraph_1
  nodes [Result()]
 }
 AnimationTransitionGraph_1 {
  path=VehicleFullBodyOverride.AnimGraphNode_StateMachine_2.CombatUpperBodyAdditive.AnimStateNode_7.Stop.AnimGraphNode_StateMachine_0.AdditiveStopTransition.AnimStateNode_3.Backwards.AnimGraphNode_StateMachine_0.AdditiveStopBackwards.AnimStateTransitionNode_1.AnimationTransitionGraph_1
  nodes [Result()]
 }
 AnimationTransitionGraph_1 {
  path=VehicleFullBodyOverride.AnimGraphNode_StateMachine_2.CombatUpperBodyAdditive.AnimStateNode_7.Stop.AnimGraphNode_StateMachine_0.AdditiveStopTransition.AnimStateNode_3.Backwards.AnimGraphNode_StateMachine_0.AdditiveStopBackwards.AnimStateTransitionNode_2.AnimationTransitionGraph_1
  nodes [Result()]
 }
 AnimationTransitionGraph_1 {
  path=VehicleFullBodyOverride.AnimGraphNode_StateMachine_2.CombatUpperBodyAdditive.AnimStateNode_7.Stop.AnimGraphNode_StateMachine_0.AdditiveStopTransition.AnimStateNode_3.Backwards.AnimGraphNode_StateMachine_0.AdditiveStopBackwards.AnimStateTransitionNode_3.AnimationTransitionGraph_1
  nodes [Result(bCanEnterTransition=true)]
 }
 Transition {
  path=VehicleFullBodyOverride.AnimGraphNode_StateMachine_2.CombatUpperBodyAdditive.AnimStateNode_7.Stop.AnimGraphNode_StateMachine_0.AdditiveStopTransition.AnimStateTransitionNode_1.Transition
  nodes [Result(bCanEnterTransition=Get Should Play Stop Backwards())]
 }
 Transition {
  path=VehicleFullBodyOverride.AnimGraphNode_StateMachine_2.CombatUpperBodyAdditive.AnimStateNode_7.Stop.AnimGraphNode_StateMachine_0.AdditiveStopTransition.AnimStateTransitionNode_3.Transition
  nodes [Result()]
 }
 Transition {
  path=VehicleFullBodyOverride.AnimGraphNode_StateMachine_2.CombatUpperBodyAdditive.AnimStateNode_7.Stop.AnimGraphNode_StateMachine_0.AdditiveStopTransition.AnimStateTransitionNode_4.Transition
  nodes [Result()]
 }
 Transition {
  path=VehicleFullBodyOverride.AnimGraphNode_StateMachine_2.CombatUpperBodyAdditive.AnimStateNode_7.Stop.AnimGraphNode_StateMachine_0.AdditiveStopTransition.AnimStateTransitionNode_5.Transition
  nodes [Result(bCanEnterTransition=true)]
 }
 JumpLand {
  path=VehicleFullBodyOverride.AnimGraphNode_StateMachine_2.CombatUpperBodyAdditive.AnimStateNode_8.JumpLand
  nodes [Output Animation Pose(Result="Sequence Player (additive) 'HardSaddleWolf_Player_CombatJumpLand_Add'"() {bLoopAnimation=false, sequence=/RidingContent/Animation/HardSaddle/Wolf/HardSaddleWolf_Player_CombatJumpLand_Add})]
 }
 TurnSettle {
  path=VehicleFullBodyOverride.AnimGraphNode_StateMachine_2.CombatUpperBodyAdditive.AnimStateNode_9.TurnSettle
  nodes [Output Animation Pose(Result=Blend Poses by bool(
     BlendPose_0="Sequence Player (additive) 'HardSaddleWolf_Player_TurnRight_Combat_Settle'"() {bLoopAnimation=false, sequence=/RidingContent/Animation/HardSaddle/Wolf/HardSaddleWolf_Player_TurnRight_Combat_Settle}
     BlendPose_1="Sequence Player (additive) 'HardSaddleWolf_Player_TurnLeft_Combat_Settle'"() {bLoopAnimation=false, sequence=/RidingContent/Animation/HardSaddle/Wolf/HardSaddleWolf_Player_TurnLeft_Combat_Settle}
     BlendTime_0=0.25
     BlendTime_1=0.25
     bActiveValue=Get Was Turning Right() -> (Was Turning Right:bool)
    ) {blendType=Linear})]
 }
 Transition {
  path=VehicleFullBodyOverride.AnimGraphNode_StateMachine_2.CombatUpperBodyAdditive.AnimStateTransitionNode_0.Transition
  nodes [Result(bCanEnterTransition=Get Is Turning in Place())]
 }
 Transition {path=VehicleFullBodyOverride.AnimGraphNode_StateMachine_2.CombatUpperBodyAdditive.AnimStateTransitionNode_1.Transition, nodes [Result()]}
 Transition {
  path=VehicleFullBodyOverride.AnimGraphNode_StateMachine_2.CombatUpperBodyAdditive.AnimStateTransitionNode_10.Transition
  nodes [Result(bCanEnterTransition=Get State Rule Moving Idle())]
 }
 Transition {
  path=VehicleFullBodyOverride.AnimGraphNode_StateMachine_2.CombatUpperBodyAdditive.AnimStateTransitionNode_11.Transition
  nodes [Result(bCanEnterTransition=Get Is Falling())]
 }
 Transition {
  path=VehicleFullBodyOverride.AnimGraphNode_StateMachine_2.CombatUpperBodyAdditive.AnimStateTransitionNode_12.Transition
  nodes [Result(bCanEnterTransition=Get State Rule Idle Moving())]
 }
 Transition {
  path=VehicleFullBodyOverride.AnimGraphNode_StateMachine_2.CombatUpperBodyAdditive.AnimStateTransitionNode_13.Transition
  nodes [Result(bCanEnterTransition=Get Is Turning in Place())]
 }
 Transition {
  path=VehicleFullBodyOverride.AnimGraphNode_StateMachine_2.CombatUpperBodyAdditive.AnimStateTransitionNode_14.Transition
  nodes [Result(bCanEnterTransition=NOT Boolean(A=Get Is Turning in Place()))]
 }
 Transition {path=VehicleFullBodyOverride.AnimGraphNode_StateMachine_2.CombatUpperBodyAdditive.AnimStateTransitionNode_15.Transition, nodes [Result()]}
 Transition {
  path=VehicleFullBodyOverride.AnimGraphNode_StateMachine_2.CombatUpperBodyAdditive.AnimStateTransitionNode_16.Transition
  nodes [Result(bCanEnterTransition=Get Is Turning in Place())]
 }
 Transition {path=VehicleFullBodyOverride.AnimGraphNode_StateMachine_2.CombatUpperBodyAdditive.AnimStateTransitionNode_18.Transition, nodes [Result()]}
 Transition {
  path=VehicleFullBodyOverride.AnimGraphNode_StateMachine_2.CombatUpperBodyAdditive.AnimStateTransitionNode_2.Transition
  nodes [Result(bCanEnterTransition=Get State Rule Idle Moving())]
 }
 Transition {
  path=VehicleFullBodyOverride.AnimGraphNode_StateMachine_2.CombatUpperBodyAdditive.AnimStateTransitionNode_3.Transition
  nodes [Result(bCanEnterTransition=Get Is Surface Swimming())]
 }
 Transition {
  path=VehicleFullBodyOverride.AnimGraphNode_StateMachine_2.CombatUpperBodyAdditive.AnimStateTransitionNode_4.Transition
  nodes [Result(bCanEnterTransition=Get Is Jumping())]
 }
 Transition {
  path=VehicleFullBodyOverride.AnimGraphNode_StateMachine_2.CombatUpperBodyAdditive.AnimStateTransitionNode_5.Transition
  nodes [Result(bCanEnterTransition=NOT Boolean(A=Get Is Surface Swimming()))]
 }
 Transition {
  path=VehicleFullBodyOverride.AnimGraphNode_StateMachine_2.CombatUpperBodyAdditive.AnimStateTransitionNode_6.Transition
  nodes [Result(bCanEnterTransition=Get State Rule Fall Idle())]
 }
 Transition {path=VehicleFullBodyOverride.AnimGraphNode_StateMachine_2.CombatUpperBodyAdditive.AnimStateTransitionNode_7.Transition, nodes [Result()]}
 Transition {
  path=VehicleFullBodyOverride.AnimGraphNode_StateMachine_2.CombatUpperBodyAdditive.AnimStateTransitionNode_8.Transition
  nodes [Result(bCanEnterTransition=Get State Rule Fall Moving())]
 }
 Transition {
  path=VehicleFullBodyOverride.AnimGraphNode_StateMachine_2.CombatUpperBodyAdditive.AnimStateTransitionNode_9.Transition
  nodes [Result(bCanEnterTransition=Get State Rule Jump Land Moving())]
 }
 MeleeOffset {
  entry=MeleeUpperBodyOffset
  path=VehicleFullBodyOverride.AnimGraphNode_StateMachine_3.MeleeOffset
  ids=id
  nodes [AnimStateNode_0 = MeleeUpperBodyOffset() {is=title AnimStateNode}]
 }
 MeleeUpperBodyOffset {
  path=VehicleFullBodyOverride.AnimGraphNode_StateMachine_3.MeleeOffset.AnimStateNode_0.MeleeUpperBodyOffset
  nodes [
   n1 = Get BodyYaw() -> (BodyYaw:float)
   n2 = Blend Poses by bool(
    BlendPose_0=Blendspace Player 'HardSaddleWolf_MeleeOffset_1H2Swing_BS'!(X=@n1) -> (Pose) {blendSpace=/RidingContent/Animation/HardSaddle/Wolf/MeleeOffset/HardSaddleWolf_MeleeOffset_1H2Swing_BS}
    BlendPose_1=Blend Poses by bool(
     BlendPose_0=Blendspace Player 'HardSaddleWolf_MeleeOffset_2HSword_BS'!(X=@n1) -> (Pose) {blendSpace=/RidingContent/Animation/HardSaddle/Wolf/MeleeOffset/HardSaddleWolf_MeleeOffset_2HSword_BS}
     BlendPose_1=Additive Identity Pose()
     bActiveValue=Get bIsTwoHandedSword() -> (bIsTwoHandedSword:bool)
    ) {blendType=Linear}
    bActiveValue=Get bIsOneHandedDualWield() -> (bIsOneHandedDualWield:bool)
   ) {blendType=Linear}
   n3 = Blendspace Player 'HardSaddleWolf_MeleeOffset_DW_BS'!(X=@n1) -> (Pose) {blendSpace=/RidingContent/Animation/HardSaddle/Wolf/MeleeOffset/HardSaddleWolf_MeleeOffset_DW_BS}
   n4 = Get bIsDualWield() -> (bIsDualWield:bool)
   n5 = Get BodyYaw()
   Output Animation Pose(Result=Apply Additive(
     Base=Blend Poses by bool(BlendPose_0=@n3, BlendPose_1=@n2, bActiveValue=@n4) {blendType=Linear}
     Additive=Blendspace Player 'HardSaddleBoar_Player_MeleeLegCorrection_BS'!(X=@n5) -> (Pose) {blendSpace=/RidingContent/Animation/HardSaddle/Boar/HardSaddleBoar_Player_MeleeLegCorrection_BS}
    ) {alphaBoolBlend {blendOption=Linear}})
  ]
 }
 EntryAdditive {
  entry=AddPose
  path=VehicleFullBodyOverride.AnimGraphNode_StateMachine_4.EntryAdditive
  ids=id
  nodes [
   AnimStateNode_0 = AddPose() {is=title AnimStateNode}
   AnimStateNode_1 = EntryAdd() {is=title AnimStateNode}
   AnimStateTransitionNode_0 = AnimStateTransitionNode() {crossfadeDuration=0.1, logicType=TLT_Inertialization}
   AnimStateTransitionNode_1 = AnimStateTransitionNode() {bAutomaticRuleBasedOnSequencePlayerInState=true, crossfadeDuration=0.3, logicType=TLT_Inertialization}
  ]
  links [
   {
    dst=EntryAdd
    rule=VehicleFullBodyOverride.AnimGraphNode_StateMachine_4.EntryAdditive.AnimStateTransitionNode_0.Transition
    src=AddPose
    via=AnimStateTransitionNode_0
   }
   {
    dst=AddPose
    rule=VehicleFullBodyOverride.AnimGraphNode_StateMachine_4.EntryAdditive.AnimStateTransitionNode_1.Transition
    src=EntryAdd
    via=AnimStateTransitionNode_1
   }
  ]
 }
 AddPose {
  path=VehicleFullBodyOverride.AnimGraphNode_StateMachine_4.EntryAdditive.AnimStateNode_0.AddPose
  nodes [Output Animation Pose(Result=Additive Identity Pose())]
 }
 EntryAdd {
  path=VehicleFullBodyOverride.AnimGraphNode_StateMachine_4.EntryAdditive.AnimStateNode_1.EntryAdd
  nodes [Output Animation Pose(Result="Sequence Player (additive) 'HardSaddleWolf_Player_SimpleEntry'"() {bLoopAnimation=false, sequence=/RidingContent/Animation/HardSaddle/Wolf/HardSaddleWolf_Player_SimpleEntry})]
 }
 Transition {
  path=VehicleFullBodyOverride.AnimGraphNode_StateMachine_4.EntryAdditive.AnimStateTransitionNode_0.Transition
  nodes [Result(bCanEnterTransition=Get bJustEntered() -> (bJustEntered:bool))]
 }
 Transition {path=VehicleFullBodyOverride.AnimGraphNode_StateMachine_4.EntryAdditive.AnimStateTransitionNode_1.Transition, nodes [Result()]}
 VehicleLowerBodyOverride {nodes [
   Input Pose() {name=InPoseLowerBodyDefault}
   Input Pose() {name=InPoseUpperBody}
   Output Pose() {layerGroup=VehicleOverride}
  ]}
 VehicleSplitBodyOverride {nodes [
   Input Pose() {name=InPoseSplitBody}
   Input Pose() {name=InPoseUpperAndLowerBody}
   Input Pose() {name=InPoseUpperAndLowerBodyRemoveRoot}
   Output Pose() {layerGroup=Vehicle Override}
  ]}
 WeaponAttachmentProceduralAdjustment {nodes [
   n1 = "Transform (Modify) Bone - Bone: pelvis"!(ComponentPose=Local To Component(LocalPose=Input Pose()), Alpha=0.25, Rotation=Get MeleeTwistRot() -> (MeleeTwistRot:Rotator)) -> (Pose) {
    rotationMode=BMM_Additive
    rotationSpace=BCS_WorldSpace
    translationSpace=BCS_WorldSpace
    alphaBoolBlend {blendOption=Linear}
    boneToModify {boneName=pelvis}
   }
   n2 = "Transform (Modify) Bone - Bone: weapon_r"!(ComponentPose=@n1, bAlphaBoolEnabled:bool=True, Scale="0.001,0.001,0.001") -> (Pose) {alphaInputType=Bool, scaleMode=BMM_Replace, alphaBoolBlend {blendOption=Linear}, boneToModify {boneName=weapon_r}, scale=<0.001 0.001 0.001>}
   n3 = "Transform (Modify) Bone - Bone: weapon_l"!(ComponentPose=@n2, bAlphaBoolEnabled:bool=True, Scale="0.001,0.001,0.001") -> (Pose) {alphaInputType=Bool, scaleMode=BMM_Replace, alphaBoolBlend {blendOption=Linear}, boneToModify {boneName=weapon_l}, scale=<0.001 0.001 0.001>}
   n4 = "Transform (Modify) Bone - Bone: hand_attach_r"!(ComponentPose=@n3, bAlphaBoolEnabled:bool=True, Scale="0.001,0.001,0.001") -> (Pose) {
    alphaInputType=Bool
    scaleMode=BMM_Replace
    alphaBoolBlend {blendOption=Linear}
    boneToModify {boneName=hand_attach_r}
    scale=<0.001 0.001 0.001>
   }
   n5 = "Transform (Modify) Bone - Bone: hand_attach_l"!(ComponentPose=@n4, bAlphaBoolEnabled:bool=True, Scale="0.001,0.001,0.001") -> (Pose) {
    alphaInputType=Bool
    scaleMode=BMM_Replace
    alphaBoolBlend {blendOption=Linear}
    boneToModify {boneName=hand_attach_l}
    scale=<0.001 0.001 0.001>
   }
   Output Pose(Result=Component To Local(ComponentPose=@n5))
  ]}
]
variables [
 AimLeanSwitchThresholdAngle {type=number, default=15}
 AimLeanSwitchThresholdAngleOffset {type=number, default=-20}
 AlphaValue {type=number}
 As Labrador Anim BP {type=Labrador_AnimBP_C, default=null}
 Backwards Play Rate {type=number}
 BodyRoll {type=number}
 CurrentCurve {type=string, default=None}
 CurrentCurveValue {type=number}
 CurrentLeanAngle {type=number}
 DisableLeanWeight {type=number}
 Is Boosting {type=boolean}
 Is Falling {type=boolean}
 Is Jumping {type=boolean}
 Is Moving Backwards {type=boolean}
 Is Surface Swimming {type=boolean}
 Is Turning in Place {type=boolean}
 IsAttacking {type=boolean}
 IsLeaningRight {type=boolean, default=true}
 IsRunning {type=boolean}
 Jump Start To Jump Loop {type=boolean}
 JustLanded {type=boolean}
 Lean from Angle {type=CurveFloat, default=&/LeanFromAngle_CurveFloat}
 LeanOffset {type=number, default=-20}
 LeanOffsetCorrectionX {type=number}
 LeanOffsetCorrectionZ {type=number}
 LeanSpineChainAngle {type=number}
 Loco To Idle {type=boolean}
 MaxLeanAngle {type=number, default=45}
 NewVar {type=Actor, default=null}
 Play Land Additive {type=boolean}
 Player Pawn Athena {type=PlayerPawn_Athena_C}
 Run Anim Position {type=number}
 Saddle Socket Transform {type=Transform, default {location=<0 0 0>, rotation=<rot 0 0 0>, scale=<1 1 1>}}
 Should Play Stop Backwards {type=boolean}
 ShouldApplyLean {type=boolean}
 ShouldPlayBoostAnim {type=boolean}
 State Rule Fall Idle {type=boolean}
 State Rule Fall Moving {type=boolean}
 State Rule Idle Moving {type=boolean}
 State Rule Jump Land Moving {type=boolean}
 State Rule Moving Idle {type=boolean}
 State Rule Swim Idle Swim Loco {type=boolean}
 TargetIsClose {type=boolean}
 TargetLeanAngle {type=number, default=45}
 Turn in Place Angle Delta {type=number}
 Turn in Place Rotation Speed {type=number}
 Walk Anim Position {type=number}
 Was Turning Right {type=boolean}
 WolfSpeed2D {type=number}
]
